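; Blueprint interface in the Battle Royale HUD minimap folder for handing an in-match feedback view model between objects.
; It declares one function, ProvideVM, with a single VM parameter typed as an Athena Ranked Bonuses VM object reference.
; The file contains only the signature, so implementers define what happens with the view model.
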(fn ProvideVM
  (params (VM :type Athena%20Ranked%20Bonuses%20VM%20Object%20Reference)))
