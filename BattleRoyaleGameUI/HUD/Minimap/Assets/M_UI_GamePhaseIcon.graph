Material /BattleRoyaleGameUI/HUD/Minimap/Assets/M_UI_GamePhaseIcon
graph {
 properties {bAutomaticallySetUsageInEditor=false, blendMode=BLEND_Translucent, materialDomain=MD_UI, usedShadingModels=Default Lit}
 nodes [
  LinearInterpolate(A=NamedRerouteUsage(), B=NamedRerouteUsage(), Alpha=NamedRerouteUsage())
  n1 = MaterialFunctionCall!(Default=ScalarParameter() {defaultValue=0.5, parameterName=BackingOpacity}, LowEnd=Constant() {r=0.3333}) {materialFunction=/BattleRoyaleGameUI/Shared/MF_UI_LowEnd_ValueSwitcher}
  n2 = ScalarParameter() {parameterName=IsFlashing}
  Multiply(
   A=NamedRerouteUsage()
   B=LinearInterpolate(A=LinearInterpolate(A=@n1.Result, Alpha=@n2) {constB=0.95}, Alpha=NamedRerouteUsage()) {constB=0.95}
  )
  n3 = Time()
  n4 = ScalarParameter() {parameterName=TimeOffset}
  n5 = ConstantBiasScale(None=NamedRerouteDeclaration(None=Add(A=@n3, B=@n4)) {name=TimeValue, nodeColor=<rgba 0 0 0 1>}) {bias=0.5}
  NamedRerouteDeclaration(None=Ceil(None=Sine(None=@n5))) {name=Gate, nodeColor=<rgba 0 0 0 1>}
  n6 = NamedRerouteUsage()
  n7 = ConstantBiasScale(None=@n6) {bias=0.25, scale=1}
  n8 = MaterialFunctionCall("Value (0-1)"=ConstantBiasScale(None=Sine(None=@n7))) {materialFunction=&MF_UI_EaseCurves}
  NamedRerouteDeclaration(None=Multiply(A=ScalarParameter() {parameterName=IsFlashing}, B=@n8.EaseInCubic)) {name=SineCurve, nodeColor=<rgba 0 0 0 1>}
  n9 = ScalarParameter() {parameterName=IsFlashing}
  n10 = ComponentMask(None=CollectionParameter() {collection=/BattleRoyaleGameUI/Shared/MPC_BRGameUI, parameterName=White}) {b=true, g=true, r=true}
  n11 = LinearInterpolate(A=@n10, B=VectorParameter() {parameterName=IconColor, defaultValue=<rgba 1 1 1 1>}.RGB, Alpha=@n9)
  n12 = LinearInterpolate(
   A=MaterialFunctionCall!(BufferSource) {materialFunction=/BattleRoyaleGameUI/Shared/MF_UI_AdaptiveColors}.Dark
   B=VectorParameter() {parameterName=BackgroundColor, defaultValue=<rgba 0.9089 0.0432 0.9219 1>}.RGB
   Alpha=@n9
  )
  n13 = NamedRerouteUsage()
  n14 = NamedRerouteUsage()
  NamedRerouteDeclaration(None=LinearInterpolate(A=@n11, B=LinearInterpolate(A=@n11, B=@n12, Alpha=@n13), Alpha=@n14)) {name=IconColor, nodeColor=<rgba 0 0 0 1>}
  NamedRerouteDeclaration(None=LinearInterpolate(A=@n12, B=LinearInterpolate(A=@n12, B=@n10, Alpha=@n13), Alpha=@n14)) {name=BackgroundColor, nodeColor=<rgba 0 0 0 1>}
  n15 = LinearInterpolate(Alpha=@n8.EaseInCubic) {constA=0.1, constB=0.5}
  n16 = ConstantBiasScale(None=Sine(None=ConstantBiasScale(None=@n6) {bias=0.2}))
  n17 = LinearInterpolate(B=LinearInterpolate(Alpha=@n16) {constA=0.2, constB=0.4}, Alpha=@n8.EaseInQuint) {constA=0.1, constB=0.8}
  n18 = ScalarParameter() {parameterName=IsFlashing}
  n19 = Add(A=ScalarParameter() {defaultValue=0.735, parameterName=IconSize}, B=Multiply(A=Multiply(A=@n15, B=@n17), B=@n18))
  n20 = Constant2Vector() {g=-0.03}
  n21 = ScalarParameter() {parameterName=IsFlashing}
  n22 = TextureObjectParameter() {parameterName=Icon, texture=/Game/Athena/UI/InGame/Creative/PropertyEditor/Icon_Picker/Textures_V2/128x128/T_UI_IconLibrary_Stormcloud_128}
  n23 = TextureSample(
   UVs=Add(
    A=MaterialFunctionCall(Texture Scale=@n19) {materialFunction=&ScaleUVsByCenter}.UVs
    B=LinearInterpolate(B=@n20, Alpha=@n21)
   )
   Tex=@n22
  ) {
   samplerSource=SSM_Clamp_WorldGroupSettings
   texture=/Game/Athena/UI/InGame/Creative/PropertyEditor/Icon_Picker/Textures_V2/64x64/T_UI_IconLibrary_Stormcloud_64
  }
  NamedRerouteDeclaration(None=@n23.A) {name=IconMask, nodeColor=<rgba 0 0 0 1>}
  n24 = Constant() {r=0.95}
  n25 = ScalarParameter() {parameterName=IsTransition}
  n26 = ScalarParameter() {defaultValue=-0.75, parameterName=Angle}
  n27 = Add(
   A=MaterialFunctionCall(
    UVs=MaterialFunctionCall() {materialFunction=&GetUserInterfaceUV}.Normalized UV
    "Rotation Angle (0-1)"=MaterialFunctionCall!(Degrees=@n26) {materialFunction=/Engine/Functions/Engine_MaterialFunctions01/Math/DegreesToRadians}.Radians
   ) {materialFunction=&CustomRotator}.Rotated Values
   B=Constant2Vector() {g=0.44}
  )
  n28 = MaterialFunctionCall!(
   DistanceFieldA=MaterialFunctionCall(Size=LinearInterpolate(A=@n24, Alpha=@n25) {constA=0.5, constB=0.5}) {materialFunction=&MF_UI_Box}.SDF
   DistanceFieldB=MaterialFunctionCall!("Texture mask (0-1 range)"=ComponentMask(None=@n27) {g=true}) {materialFunction=/FortUILibrary/MaterialFunctions/Shapes/Modifiers/MF_UI_SDF_RemapTextureRange}.SDF
   BlendAmount
  ) {materialFunction=/FortUILibrary/MaterialFunctions/Shapes/Modifiers/MF_UI_SDF_Subtraction}
  n29 = LinearInterpolate(Alpha=@n16) {constA=0.25, constB=0.5}
  n30 = Multiply(
   A=LinearInterpolate(B=@n29, Alpha=@n8.EaseInQuint) {constA=0.2}
   B=LinearInterpolate(B=@n29, Alpha=ConstantBiasScale(None=@n8.EaseInQuint)) {constA=0.25}
  )
  n31 = MaterialFunctionCall(Size=LinearInterpolate(Alpha=Multiply(A=@n18, B=@n30)) {constA=0.9, constB=1.2}) {materialFunction=&MF_UI_Circle}
  n32 = ScalarParameter() {parameterName=IsTransition}
  NamedRerouteDeclaration(None=MaterialFunctionCall(DistanceField=LinearInterpolate(A=@n28.Exact, B=@n31.SDF, Alpha=@n32)) {materialFunction=&MF_UI_AntiAliasing}.Inner) {name=BackgroundMask, nodeColor=<rgba 0 0 0 1>}
 ]
}
