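# UI material that draws animated effects for the Battle Royale minimap's game phase icon.
# SDF rings, circle and radiating glow follow eased animation parameters, or loop on time via IsLoopingRingAndCircle.
# Pulse-warped grids of random boxes, a storm tech pattern and panned radial noise fog add digital detail within circular masks.
# The layers drive a three-color gradient with intensity boost, alpha-composited using an additive ratio.
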Material /BattleRoyaleGameUI/HUD/Minimap/Assets/M_UI_GamePhaseIconFX
graph {
 properties {bAutomaticallySetUsageInEditor=false, blendMode=BLEND_AlphaComposite, materialDomain=MD_UI, usedShadingModels=Default Lit}
 nodes [
  n1 = Constant()
  n2 = MaterialFunctionCall!(
   UV=NamedRerouteUsage()
   Global Offset
   Global Rotation
   Tiling=Multiply(A=ScalarParameter() {defaultValue=6, parameterName=DigiTiling}) {constB=2}
   Cell Offset=@n1
   Cell Scale=Constant() {r=1}
   Cell Rotation=@n1
   Random Seed
  ) {materialFunction=/FortUILibrary/MaterialFunctions/Basics/MF_UI_ParticleGrid}
  n3 = NamedRerouteUsage()
  n4 = MaterialFunctionCall!(UV=@n2.Random Cell Values, Seed=Multiply(A=@n3) {constB=0.333}) {materialFunction=&MF_UI_Random}
  n5 = Saturate(None=Subtract(A=@n4.Result) {constB=0.75})
  n6 = MaterialFunctionCall(
   UVs=@n2.Cells UV
   Size=MaterialFunctionCall(UV=@n2.Random Cell Values, Seed=Multiply(A=@n3) {constB=0.45}) {materialFunction=&MF_UI_Random}.Result
   CornerRoundness=Constant() {r=0.2}
   OutlineThickness=Constant() {r=0.05}
  ) {materialFunction=&MF_UI_Box}
  n7 = Min(
   A=MaterialFunctionCall(UVs=@n2.Cells UV, Size=@n4.Result, CornerRoundness=Constant() {r=0.1}) {materialFunction=&MF_UI_Box}.Fill
   B=@n5
  )
  n8 = Constant()
  n9 = MaterialFunctionCall(
   UV=NamedRerouteUsage()
   Tiling=ScalarParameter() {defaultValue=6, parameterName=DigiTiling}
   Cell Offset=@n8
   Cell Scale=Constant() {r=1}
   Cell Rotation=@n8
  ) {materialFunction=&MF_UI_ParticleGrid}
  n10 = NamedRerouteUsage()
  n11 = MaterialFunctionCall(UV=@n9.Random Cell Values, Seed=Multiply(A=@n10) {constB=0.25}) {materialFunction=&MF_UI_Random}
  n12 = MaterialFunctionCall(UVs=@n9.Cells UV, Size=@n11.Result, CornerRoundness=Constant() {r=0.2}) {materialFunction=&MF_UI_Box}
  n13 = Multiply(A=OneMinus(None=NamedRerouteUsage())) {constB=0.5}
  n14 = Saturate(None=Subtract(A=@n11.Result) {constB=0.25})
  n15 = Multiply(A=Min(
    A=MaterialFunctionCall(DistanceField=@n12.SDF, Smoothing=@n13) {materialFunction=&MF_UI_AntiAliasing}.Inner
    B=@n14
   )) {constB=0.125}
  n16 = Saturate(None=Subtract(A=@n11.Result) {constB=0.5})
  n17 = MaterialFunctionCall(
   UVs=@n9.Cells UV
   Size=MaterialFunctionCall(UV=@n9.Random Cell Values, Seed=Multiply(A=@n10) {constB=0.35}) {materialFunction=&MF_UI_Random}.Result
   CornerRoundness=Constant() {r=0.2}
   OutlineThickness=Constant() {r=0.05}
  ) {materialFunction=&MF_UI_Box}
  n18 = Multiply(A=Min(
    A=OneMinus(None=@n16)
    B=MaterialFunctionCall(DistanceField=@n17.SDFOutline, Smoothing=@n13) {materialFunction=&MF_UI_AntiAliasing}.Inner
   )) {constB=0.25}
  Add(
   A=LinearInterpolate(A=Min(A=@n5, B=@n6.Outline), Alpha=Multiply(A=@n7) {constB=0.5})
   B=LinearInterpolate(A=Multiply(A=@n15, B=@n14), Alpha=Multiply(A=@n16, B=@n18))
  )
  n19 = NamedRerouteUsage()
  Add(A=NamedRerouteUsage(), B=@n19)
  n20 = NamedRerouteUsage()
  Add(A=NamedRerouteUsage(), B=@n20)
  n21 = VectorParameter() {parameterName=GlowColor, defaultValue=<rgba 1 0.0097 0.8388 1>}
  n22 = Add(A=NamedRerouteUsage(), B=NamedRerouteUsage())
  n23 = Add(A=LinearInterpolate(A=@n22, B=@n22, Alpha=NamedRerouteUsage()), B=NamedRerouteUsage())
  n24 = Multiply(A=LinearInterpolate(Alpha=NamedRerouteUsage()) {constA=0.025, constB=0.25}, B=NamedRerouteUsage()) {constB=0.25}
  n25 = Add(A=NamedRerouteUsage(), B=Add(A=NamedRerouteUsage(), B=@n24))
  n26 = MaterialFunctionCall(
   A=Constant()
   B=VectorParameter() {parameterName=BackgroundColor, defaultValue=<rgba 0.1667 0 1 1>}.RGB
   C=@n21.RGB
   Alpha=LinearInterpolate(A=Add(A=@n23, B=@n25), B=@n19, Alpha=ScalarParameter() {parameterName=IsLoopingRingAndCircle})
  ) {materialFunction=&Lerp_3Color}
  n27 = Multiply(A=@n26.Result) {constB=8}
  n28 = ScalarParameter() {parameterName=Anim_Intensity}
  n29 = Multiply(A=@n21.RGB, B=SmoothStep(Value=@n24))
  n30 = ScalarParameter() {parameterName=IsLoopingRingAndCircle}
  Multiply(A=VertexColor().A, B=Add(A=LinearInterpolate(A=@n26.Result, B=@n27, Alpha=@n28), B=LinearInterpolate(A=@n29, Alpha=@n30) {constB=0}))
  Multiply(A=@n22, B=NamedRerouteUsage())
  Multiply(
   A=LinearInterpolate(A=NamedRerouteUsage(), B=@n20, Alpha=ScalarParameter() {parameterName=IsLoopingRingAndCircle})
   B=OneMinus(None=ScalarParameter() {defaultValue=0.5, parameterName=AddtiveRatio})
  )
  n31 = Constant()
  n32 = Saturate(None=OneMinus(None=MaterialFunctionCall!(UVs, Radius=@n31) {materialFunction=/FortUILibrary/MaterialFunctions/Shapes/SDFs/MF_SDF_Circle}.SDF))
  n33 = MaterialFunctionCall("Value (0-1)"=@n32) {materialFunction=&MF_UI_EaseCurves}
  NamedRerouteDeclaration(None=Multiply(A=@n33.EaseInSine, B=@n33.EaseInSine)) {name=GlowMask, nodeColor=<rgba 0 0 0 1>}
  NamedRerouteDeclaration(None=Multiply(A=@n33.EaseInSine, B=@n33.EaseInSine)) {name=DigiFXMask, nodeColor=<rgba 0 0 0 1>}
  NamedRerouteDeclaration(None=@n32) {name=GlowMask_Linear, nodeColor=<rgba 0 0 0 1>}
  n34 = NamedRerouteDeclaration(None=Add(A=Time(), B=ScalarParameter() {parameterName=TimeOffset}) {constB=0.2}) {name=TimeValue, nodeColor=<rgba 0 0 0 1>}
  n35 = LinearInterpolate(B=ScalarParameter() {defaultValue=0.3333, parameterName=DigiTilesPulseSize}, Alpha=ConstantBiasScale(None=Sine(None=@n34) {period=2})) {constB=0.25}
  n36 = MaterialFunctionCall("Value (0-1)"=Sine(None=@n34)) {materialFunction=&MF_UI_EaseCurves}
  n37 = ConstantBiasScale(None=@n36.EaseInQuint)
  NamedRerouteDeclaration(None=Multiply(A=LinearInterpolate(B=@n35, Alpha=@n37), B=LinearInterpolate(B=@n35, Alpha=@n36.EaseInQuint))) {name=ParticleGridWarpAnim, nodeColor=<rgba 0 0 0 1>}
  n38 = ConstantBiasScale(None=Sine(None=Add(A=@n34) {constB=0.25}))
  NamedRerouteDeclaration(None=OneMinus(None=MaterialFunctionCall("Value (0-1)"=@n38) {materialFunction=&MF_UI_EaseCurves}.EaseInSine)) {name=PulsingGlowAnim_Secondary, nodeColor=<rgba 0 0 0 1>}
  n39 = MaterialFunctionCall!(Vector or UVs) {materialFunction=/FortUILibrary/MaterialFunctions/Basics/MF_UI_CheaperVectorToRadial}
  n40 = Panner(Coordinate=Multiply(A=@n39.Radial Coordinates) {constB=3}) {speedX=0.5, speedY=-0.5}
  n41 = DDX(Value=@n39.Radial Coordinates)
  n42 = DDY(Value=@n39.Radial Coordinates)
  NamedRerouteDeclaration(None=MaterialFunctionCall("Value (0-1)"=TextureSample!(UVs=@n40, Tex, "DDX(UVs)"=@n41, "DDY(UVs)"=@n42, Apply View MipBias) {mipValueMode=TMVM_Derivative, texture=&/T_Noise_Caustics}.G) {materialFunction=&MF_UI_EaseCurves}.EaseInOutCubic) {name=NoiseMask, nodeColor=<rgba 0 0 0 1>}
  n43 = Multiply(A=Time(), B=ScalarParameter() {defaultValue=1, parameterName=LoopingSpeed})
  n44 = Constant() {r=1}
  n45 = MaterialFunctionCall("Value (0-1)"=LinearInterpolate(A=ScalarParameter() {parameterName=FullCircleAnim}, B=Fmod(A=@n43, B=@n44), Alpha=ScalarParameter() {parameterName=IsLoopingRingAndCircle})) {materialFunction=&MF_UI_EaseCurves}
  n46 = MaterialFunctionCall(Radius=Multiply(A=@n45.EaseOutQuint) {constB=0.6}) {materialFunction=&MF_SDF_Circle}
  n47 = LinearInterpolate(Alpha=@n45.EaseOutSine) {constB=0.05}
  NamedRerouteDeclaration(None=Multiply(
    A=LinearInterpolate(Alpha=@n45.EaseOutQuint) {constA=0.25, constB=0}
    B=MaterialFunctionCall(DistanceField=@n46.SDF, Smoothing=@n47) {materialFunction=&MF_UI_AntiAliasing}.Inner
   )) {name=FullCircle, nodeColor=<rgba 0 0 0 1>}
  n48 = MaterialFunctionCall("Value (0-1)"=ScalarParameter() {parameterName=SecondRingAnim}) {materialFunction=&MF_UI_EaseCurves}
  n49 = ScalarParameter() {defaultValue=0.0125, parameterName=RingWidth}
  n50 = MaterialFunctionCall(
   DistanceField=MaterialFunctionCall(
    Size=Multiply(A=@n48.EaseOutQuint) {constB=0.9}
    OutlineThickness=LinearInterpolate(A=@n49, Alpha=@n48.EaseInOutCubic) {constA=0.025, constB=-0.05}
   ) {materialFunction=&MF_UI_Circle}.SDFOutline
   Smoothing=LinearInterpolate(Alpha=@n48.EaseInOutCubic) {constA=0.01, constB=0.1}
  ) {materialFunction=&MF_UI_AntiAliasing}
  NamedRerouteDeclaration(None=Multiply(A=LinearInterpolate(Alpha=@n48.EaseOutQuint) {constA=3, constB=0}, B=Multiply(A=@n48.EaseOutQuint, B=@n50.Inner))) {name=SecondRing, nodeColor=<rgba 0 0 0 1>}
  n51 = Multiply(A=Add(A=Time()) {constB=0.25}, B=ScalarParameter() {defaultValue=1, parameterName=LoopingSpeed})
  n52 = Constant() {r=1}
  n53 = MaterialFunctionCall("Value (0-1)"=LinearInterpolate(A=ScalarParameter() {parameterName=FirstRingAnim}, B=Fmod(A=@n51, B=@n52), Alpha=ScalarParameter() {parameterName=IsLoopingRingAndCircle})) {materialFunction=&MF_UI_EaseCurves}
  n54 = ScalarParameter() {defaultValue=0.0125, parameterName=RingWidth}
  n55 = MaterialFunctionCall(
   DistanceField=MaterialFunctionCall(
    Size=Multiply(A=@n53.EaseOutQuint) {constB=0.9}
    OutlineThickness=LinearInterpolate(A=@n54, Alpha=@n53.EaseInOutCubic) {constA=0.025, constB=-0.05}
   ) {materialFunction=&MF_UI_Circle}.SDFOutline
   Smoothing=LinearInterpolate(Alpha=@n53.EaseInOutCubic) {constA=0.025, constB=0.1}
  ) {materialFunction=&MF_UI_AntiAliasing}
  NamedRerouteDeclaration(None=Multiply(A=LinearInterpolate(Alpha=@n53.EaseOutQuint) {constA=0.5, constB=0}, B=Multiply(A=@n53.EaseOutQuint, B=@n55.Inner))) {name=FirstRing, nodeColor=<rgba 0 0 0 1>}
  n56 = NamedRerouteUsage()
  n57 = LinearInterpolate(Alpha=NamedRerouteUsage()) {constA=3, constB=7}
  n58 = Saturate(None=MaterialFunctionCall() {materialFunction=/BattleRoyaleGameUI/Shared/FX/MF_UI_BR_StormTechPattern}.Result)
  n59 = Multiply(A=Multiply(A=NamedRerouteUsage()) {constB=2}, B=@n58)
  n60 = Multiply(A=@n58, B=NamedRerouteUsage())
  NamedRerouteDeclaration(None=Multiply(A=Multiply(A=@n56, B=@n57) {constB=8}, B=Add(A=@n59, B=@n60))) {name=DigiFX, nodeColor=<rgba 0 0 0 1>}
  n61 = TextureCoordinate()
  n62 = ConstantBiasScale(None=@n61) {bias=-0.5, scale=-2}
  n63 = Divide(A=DotProduct(A=@n62, B=@n62), B=ScalarParameter() {defaultValue=0.25, parameterName=WarpRadius})
  n64 = ScalarParameter() {defaultValue=0.05, parameterName=WarpMin}
  n65 = ScalarParameter() {defaultValue=0.5, parameterName=WarpMax}
  n66 = NamedRerouteUsage()
  n67 = Saturate(None=Multiply(A=OneMinus(None=@n63), B=LinearInterpolate(A=@n64, B=@n65, Alpha=@n66)))
  NamedRerouteDeclaration(None=Add(A=@n61, B=Multiply(A=@n62, B=@n67))) {name=ParticleGridUVs, nodeColor=<rgba 0 0 0 1>}
  NamedRerouteDeclaration(None=@n33.EaseInQuint) {name=GlowMask_Center, nodeColor=<rgba 0 0 0 1>}
  NamedRerouteDeclaration(None=@n48.EaseOutQuint) {name=AnimRingDissolve, nodeColor=<rgba 0 0 0 1>}
  n68 = MaterialFunctionCall("Value (0-1)"=ScalarParameter() {parameterName=PulseGlow}) {materialFunction=&MF_UI_EaseCurves}
  NamedRerouteDeclaration(None=@n68.EaseOutSine) {name=PulseAnim, nodeColor=<rgba 0 0 0 1>}
  n69 = LinearInterpolate(Alpha=@n68.EaseOutCubic)
  n70 = Multiply(
   A=SmoothStep(Min=@n69, Value=MaterialFunctionCall(Radius=@n69) {materialFunction=&MF_SDF_Circle}.SDF) {constMax=0, constMin=1}
   B=@n68.EaseOutQuint
  )
  n71 = OneMinus(None=@n68.EaseOutQuint)
  NamedRerouteDeclaration(None=MaterialFunctionCall("Value (0-1)"=Multiply(A=@n70, B=@n71)) {materialFunction=&MF_UI_EaseCurves}.EaseOutQuint) {name=RadiatingGlow, nodeColor=<rgba 0 0 0 1>}
  NamedRerouteDeclaration(None=@n37) {name=PulsingGlowAnim, nodeColor=<rgba 0 0 0 1>}
  n72 = MaterialFunctionCall() {materialFunction=&MF_UI_CheaperVectorToRadial}
  n73 = Multiply(A=Time()) {constB=0.5}
  n74 = DDX(Value=@n72.Radial Coordinates)
  n75 = DDY(Value=@n72.Radial Coordinates)
  n76 = TextureSample(
   UVs=Panner(Coordinate=Multiply(A=@n72.Radial Coordinates) {constB=3}, Time=@n73) {speedX=0.05, speedY=-0.25}
   "DDX(UVs)"=@n74
   "DDY(UVs)"=@n75
  ) {mipValueMode=TMVM_Derivative, samplerType=SAMPLERTYPE_LinearGrayscale, texture=&/T_Noise_Ripples}
  n77 = TextureSample(
   UVs=Panner(Coordinate=Multiply(A=@n72.Radial Coordinates) {constB=2}, Time=@n73) {speedX=-0.025, speedY=-0.05}
   "DDX(UVs)"=@n74
   "DDY(UVs)"=@n75
  ) {mipValueMode=TMVM_Derivative, samplerType=SAMPLERTYPE_LinearGrayscale, texture=&T_Noise_Ripples}
  n78 = Saturate(None=Multiply(A=Multiply(A=@n76.G, B=@n77.G)) {constB=4})
  n79 = MaterialFunctionCall("Value (0-1)"=NamedRerouteUsage()) {materialFunction=&MF_UI_EaseCurves}
  n80 = NamedRerouteUsage()
  NamedRerouteDeclaration(None=Multiply(
    A=MaterialFunctionCall("Value (0-1)"=@n78) {materialFunction=&MF_UI_EaseCurves}.EaseInSine
    B=LinearInterpolate(A=@n79.EaseInSine, B=@n79.EaseInSine, Alpha=@n80)
   )) {name=CenterFog, nodeColor=<rgba 0 0 0 1>}
 ]
}
