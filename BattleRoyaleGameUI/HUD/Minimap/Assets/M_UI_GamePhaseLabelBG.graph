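# UI material for the minimap's game phase label background, drawing a colored gradient glow with animatable intensity.
# Curve atlas gradients sampled along horizontal and mirrored vertical 9-slice UVs multiply into a mask, which IsHardEdge blends toward a harder-edged variant.
# Easing functions shape that mask for color and opacity.
# A static switch selects alpha-composite output with an additive ratio, or a primary-to-secondary color blend tinted toward shared white.
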Material /BattleRoyaleGameUI/HUD/Minimap/Assets/M_UI_GamePhaseLabelBG
graph {
 properties {bAutomaticallySetUsageInEditor=false, blendMode=BLEND_AlphaComposite, materialDomain=MD_UI, usedShadingModels=Default Lit}
 nodes [
  n1 = MaterialFunctionCall() {materialFunction=&GetUserInterfaceUV}
  n2 = ConstantBiasScale(None=NamedRerouteDeclaration(None=ComponentMask(None=@n1.9-Slice UV) {g=true}) {name=VerticalMask, nodeColor=<rgba 0 0 0 1>}) {bias=-0.5, scale=2}
  n3 = CurveAtlasRowParameter(CurveTime=OneMinus(None=Abs(None=@n2))) {
   atlas=/BattleRoyaleGameUI/Shared/UICurveAtlas_BRGameUI
   curve=/BattleRoyaleGameUI/Shared/UICurve_BRGameUI_GamePhaseLabel
   defaultValue=2
   parameterName=VerticalGradient
  }
  n4 = CurveAtlasRowParameter(CurveTime=NamedRerouteDeclaration(None=ComponentMask(None=@n1.9-Slice UV) {r=true}) {name=HorizontalMask, nodeColor=<rgba 0 0 0 1>}) {
   atlas=&UICurveAtlas_BRGameUI
   curve=&UICurve_BRGameUI_GamePhaseLabel
   defaultValue=2
   parameterName=HorizontalGradient
  }
  n5 = LinearInterpolate(
   A=Multiply(A=@n3.G, B=@n4.R)
   B=Multiply(
    A=MaterialFunctionCall("Value (0-1)"=@n4.R) {materialFunction=&MF_UI_EaseCurves}.EaseOutCubic
    B=NamedRerouteUsage()
   )
   Alpha=ScalarParameter() {parameterName=IsHardEdge}
  )
  n6 = ScalarParameter() {defaultValue=1, parameterName=AdditiveRatio, sliderMax=1}
  n7 = MaterialFunctionCall("Value (0-1)"=@n5) {materialFunction=&MF_UI_EaseCurves}
  StaticSwitchParameter(True=Multiply(A=@n5, B=OneMinus(None=@n6)), False=Saturate(None=@n7.EaseOutSine)) {defaultValue=true, parameterName=IsAlphaComposite}
  n8 = VectorParameter() {parameterName=SecondaryColor, defaultValue=<rgba 0.1651 0 1 1>}
  n9 = LinearInterpolate(A=@n8.RGB, B=LinearInterpolate(A=Multiply(A=@n8.RGB) {constB=8}), Alpha=ScalarParameter() {parameterName=Anim_Intensity})
  n10 = Saturate(None=@n7.EaseInCubic)
  n11 = VectorParameter() {parameterName=PrimaryColor, defaultValue=<rgba 0.1651 0 1 1>}
  n12 = ScalarParameter() {parameterName=Anim_Intensity}
  n13 = CollectionParameter() {collection=/BattleRoyaleGameUI/Shared/MPC_BRGameUI, parameterName=White}
  StaticSwitchParameter(
   True=Multiply(A=VertexColor().A, B=Multiply(A=@n9, B=@n10))
   False=LinearInterpolate(A=LinearInterpolate(A=@n11.RGB, B=@n8.RGB, Alpha=@n12), B=ComponentMask(None=@n13) {b=true, g=true, r=true}, Alpha=@n7.EaseInCubic)
  ) {defaultValue=true, parameterName=IsAlphaComposite}
 ]
}
