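# HUD minimap widget for Battle Royale that shows the game phase icon and announces phase changes with an animated text alert, hidden during Setup and Warmup.
# OnUpdateGamePhaseIcon caches the step, applies its text, textures and colors through dynamic materials, and toggles icon flashing.
# Alerts broadcast Updated Game Phase Step and fade on a timer, with separate touch animations chosen by input type.
# Construct applies compact touch styling and squad-size-based icon skew.
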WidgetBlueprint /BattleRoyaleGameUI/HUD/Minimap/WBP_BRGameUI_GamePhaseWidget
animations [
 AnimTransitionFlashing {
  binding_count=1
  length_seconds=0.25
  object=&self:AnimTransitionFlashing
  sequence {bindings [Image_Icon {
     id=#1
     possessed_class=Image
     tracks [{
       type=MovieSceneWidgetMaterialTrack
       sections [{
         type=MovieSceneParameterSection
         end_frame=5
         start_frame=0
         channels [IsFlashing {type=MovieSceneScriptingFloatChannel, keys [[0,0],[5,1]]}]
        }]
      }]
    }]}
 }
 AnimAlertShow {
  binding_count=10
  length_seconds=3
  object=&self:AnimAlertShow
  sequence {
   bindings [
    Text_PhaseAlert {
     id=#2
     possessed_class=TextBlock
     tracks [
      {
       type=MovieSceneByteTrack
       display_name=Visibility
       sections [{type=MovieSceneByteSection, channels [None {type=MovieSceneScriptingByteChannel, const [3,1]}]}]
      }
      {
       type=MovieSceneFloatTrack
       display_name=Render Opacity
       sections [{type=MovieSceneFloatSection, channels [None {type=MovieSceneScriptingFloatChannel, keys [[27,0],[36,1]]}]}]
      }
      {
       type=MovieScene2DTransformTrack
       sections [{
         type=MovieScene2DTransformSection
         channels [
          Translation.X {type=MovieSceneScriptingFloatChannel, keys [[21,128],[45,-24]]}
          Translation.Y {type=MovieSceneScriptingFloatChannel, default=0}
          Angle {type=MovieSceneScriptingFloatChannel, default=0}
          Scale.X {type=MovieSceneScriptingFloatChannel, default=1, keys [[18,1],[25,1.25],[71,1]]}
          Scale.Y {type=MovieSceneScriptingFloatChannel, default=1, keys [[18,1],[25,1.25],[71,1]]}
          Shear.X {type=MovieSceneScriptingFloatChannel, default=0}
          Shear.Y {type=MovieSceneScriptingFloatChannel, default=0}
         ]
        }]
      }
     ]
    }
    SizeBox_RemainingTime {
     id=#3
     possessed_class=SizeBox
     tracks [{
       type=MovieScene2DTransformTrack
       sections [{
         type=MovieScene2DTransformSection
         channels [
          Translation.X {type=MovieSceneScriptingFloatChannel, keys [[3,0],[12,-10]]}
          Translation.Y {type=MovieSceneScriptingFloatChannel, const [0,1]}
          Angle {type=MovieSceneScriptingFloatChannel, const [0,1]}
          Scale.X {type=MovieSceneScriptingFloatChannel, keys [[3,1],[12,1.2]]}
          Scale.Y {type=MovieSceneScriptingFloatChannel, keys [[3,1],[12,1.2]]}
          Shear.X {type=MovieSceneScriptingFloatChannel, const [0,1]}
          Shear.Y {type=MovieSceneScriptingFloatChannel, const [0,1]}
         ]
        }]
      }]
    }
    Image_IconFX_LoopingPulse {
     id=#4
     possessed_class=Image
     tracks [
      {
       type=MovieSceneFloatTrack
       display_name=Render Opacity
       sections [{type=MovieSceneFloatSection, channels [None {type=MovieSceneScriptingFloatChannel, keys [[54,0],[60,1]]}]}]
      }
      {
       type=MovieSceneByteTrack
       display_name=Visibility
       sections [{type=MovieSceneByteSection, channels [None {type=MovieSceneScriptingByteChannel, const [3,1]}]}]
      }
     ]
    }
    Image_Icon {
     id=#5
     possessed_class=Image
     tracks [
      {
       type=MovieScene2DTransformTrack
       sections [{
         type=MovieScene2DTransformSection
         channels [
          Scale.X {type=MovieSceneScriptingFloatChannel, keys [[0,1],[18,1.25]]}
          Scale.Y {type=MovieSceneScriptingFloatChannel, keys [[0,1],[18,1.25]]}
         ]
        }]
      }
      {
       type=MovieSceneWidgetMaterialTrack
       sections [{type=MovieSceneParameterSection, channels [IsTransition {type=MovieSceneScriptingFloatChannel, keys [[0,0],[6,1]]}]}]
      }
     ]
    }
    Image_MapEdgeGlow {
     id=#6
     possessed_class=Image
     tracks [
      {
       type=MovieSceneByteTrack
       display_name=Visibility
       sections [{type=MovieSceneByteSection, channels [None {type=MovieSceneScriptingByteChannel, const [3,1]}]}]
      }
      {
       type=MovieSceneFloatTrack
       display_name=Render Opacity
       sections [{type=MovieSceneFloatSection, channels [None {type=MovieSceneScriptingFloatChannel, keys [[90,0],[180,0.25]]}]}]
      }
     ]
    }
    Image_IconFX {
     id=#7
     possessed_class=Image
     tracks [
      {
       type=MovieSceneByteTrack
       display_name=Visibility
       sections [{type=MovieSceneByteSection, channels [None {type=MovieSceneScriptingByteChannel, const [3,1]}]}]
      }
      {
       type=MovieSceneFloatTrack
       display_name=Render Opacity
       sections [{type=MovieSceneFloatSection, channels [None {type=MovieSceneScriptingFloatChannel, keys [[0,0],[3,1],[90,1],[180,0.25]]}]}]
      }
      {
       type=MovieSceneWidgetMaterialTrack
       sections [{
         type=MovieSceneParameterSection
         channels [
          FirstRingAnim {type=MovieSceneScriptingFloatChannel, keys [[0,0],[120,1]]}
          SecondRingAnim {type=MovieSceneScriptingFloatChannel, keys [[0,0],[180,1]]}
          FullCircleAnim {type=MovieSceneScriptingFloatChannel, keys [[12,0],[150,1]]}
          Anim_Intensity {type=MovieSceneScriptingFloatChannel, keys [[0,0],[24,1],[180,0]]}
         ]
        }]
      }
      {
       type=MovieScene2DTransformTrack
       sections [{
         type=MovieScene2DTransformSection
         channels [
          Scale.X {type=MovieSceneScriptingFloatChannel, keys [[0,1],[150,10]]}
          Scale.Y {type=MovieSceneScriptingFloatChannel, keys [[0,1],[150,10]]}
         ]
        }]
      }
     ]
    }
    Image_AlertBackground_Glow {
     id=#8
     possessed_class=Image
     tracks [
      {
       type=MovieSceneFloatTrack
       display_name=Render Opacity
       sections [{type=MovieSceneFloatSection, channels [None {type=MovieSceneScriptingFloatChannel, keys [[7,0],[22,1],[142,0]]}]}]
      }
      {
       type=MovieSceneByteTrack
       display_name=Visibility
       sections [{type=MovieSceneByteSection, channels [None {type=MovieSceneScriptingByteChannel, const [3,1]}]}]
      }
      {
       type=MovieScene2DTransformTrack
       sections [{
         type=MovieScene2DTransformSection
         channels [
          Scale.X {type=MovieSceneScriptingFloatChannel, keys [[7,0.5],[52,2.5]]}
          Scale.Y {type=MovieSceneScriptingFloatChannel, keys [[7,1],[19,5],[97,3]]}
         ]
        }]
      }
     ]
    }
    HorizontalBox_Content {
     id=#9
     display_name=HorizontalBox_0
     possessed_class=HorizontalBox
     tracks [{
       type=MovieScene2DTransformTrack
       sections [{
         type=MovieScene2DTransformSection
         channels [
          Angle {type=MovieSceneScriptingFloatChannel, default=0}
          Scale.X {type=MovieSceneScriptingFloatChannel, keys [[0,1],[24,1.2]]}
          Scale.Y {type=MovieSceneScriptingFloatChannel, keys [[0,1],[24,1.2]]}
          Shear.X {type=MovieSceneScriptingFloatChannel, default=0}
          Shear.Y {type=MovieSceneScriptingFloatChannel, default=0}
         ]
        }]
      }]
    }
    WBP_BRGameUI_GamePhaseWidget_C_147 {
     id=#10
     possessed_class=WBP_BRGameUI_GamePhaseWidget_C
     tracks [{
       type=MovieScene2DTransformTrack
       sections [{
         type=MovieScene2DTransformSection
         channels [
          Translation.X {type=MovieSceneScriptingFloatChannel, default=0}
          Translation.Y {type=MovieSceneScriptingFloatChannel, keys [[0,0],[24,12]]}
         ]
        }]
      }]
    }
    Image_AlertBackground {
     id=#11
     possessed_class=Image
     tracks [
      {
       type=MovieSceneFloatTrack
       display_name=Render Opacity
       sections [{type=MovieSceneFloatSection, channels [None {type=MovieSceneScriptingFloatChannel, keys [[12,0],[20,1],[30,1],[90,0]]}]}]
      }
      {
       type=MovieSceneByteTrack
       display_name=Visibility
       sections [{type=MovieSceneByteSection, channels [None {type=MovieSceneScriptingByteChannel, const [3,1]}]}]
      }
      {
       type=MovieScene2DTransformTrack
       sections [{
         type=MovieScene2DTransformSection
         channels [
          Translation.X {type=MovieSceneScriptingFloatChannel, const [0,1]}
          Translation.Y {type=MovieSceneScriptingFloatChannel, const [0,1]}
          Angle {type=MovieSceneScriptingFloatChannel, const [0,1]}
          Scale.X {type=MovieSceneScriptingFloatChannel, keys [[0,0],[30,2]]}
          Scale.Y {type=MovieSceneScriptingFloatChannel, keys [[0,1.25],[15,1.5],[90,1]]}
          Shear.X {type=MovieSceneScriptingFloatChannel, const [0,1]}
          Shear.Y {type=MovieSceneScriptingFloatChannel, const [0,1]}
         ]
        }]
      }
      {
       type=MovieSceneWidgetMaterialTrack
       sections [{type=MovieSceneParameterSection, channels [Anim_Intensity {type=MovieSceneScriptingFloatChannel, keys [[6,0],[30,1],[90,0]]}]}]
      }
     ]
    }
   ]
   display_rate {numerator=60}
  }
 }
 AnimAlertHide {
  binding_count=9
  length_seconds=0.5
  object=&self:AnimAlertHide
  sequence {
   bindings [
    Text_PhaseAlert {
     id=#2
     possessed_class=TextBlock
     tracks [
      {
       type=MovieSceneByteTrack
       display_name=Visibility
       sections [{type=MovieSceneByteSection, channels [None {type=MovieSceneScriptingByteChannel, keys [[0,3],[30,2]]}]}]
      }
      {
       type=MovieSceneFloatTrack
       display_name=Render Opacity
       sections [{type=MovieSceneFloatSection, channels [None {type=MovieSceneScriptingFloatChannel, keys [[0,1],[12,0]]}]}]
      }
      {
       type=MovieScene2DTransformTrack
       sections [{
         type=MovieScene2DTransformSection
         channels [
          Translation.X {type=MovieSceneScriptingFloatChannel, default=0, keys [[0,-24],[30,128]]}
          Translation.Y {type=MovieSceneScriptingFloatChannel, default=0}
          Angle {type=MovieSceneScriptingFloatChannel, default=0}
          Scale.X {type=MovieSceneScriptingFloatChannel, keys [[0,1],[30,0.9]]}
          Scale.Y {type=MovieSceneScriptingFloatChannel, keys [[0,1],[30,0.9]]}
          Shear.X {type=MovieSceneScriptingFloatChannel, default=0}
          Shear.Y {type=MovieSceneScriptingFloatChannel, default=0}
         ]
        }]
      }
     ]
    }
    HorizontalBox_Content {
     id=#12
     display_name=HorizontalBox_0
     possessed_class=HorizontalBox
     tracks [{
       type=MovieScene2DTransformTrack
       sections [{
         type=MovieScene2DTransformSection
         channels [
          Angle {type=MovieSceneScriptingFloatChannel, default=0}
          Scale.X {type=MovieSceneScriptingFloatChannel, keys [[0,1.2],[30,1]]}
          Scale.Y {type=MovieSceneScriptingFloatChannel, keys [[0,1.2],[30,1]]}
          Shear.X {type=MovieSceneScriptingFloatChannel, default=0}
          Shear.Y {type=MovieSceneScriptingFloatChannel, default=0}
         ]
        }]
      }]
    }
    Image_MapEdgeGlow {
     id=#13
     possessed_class=Image
     tracks [{
       type=MovieSceneFloatTrack
       display_name=Render Opacity
       sections [{type=MovieSceneFloatSection, channels [None {type=MovieSceneScriptingFloatChannel, keys [[0,0.25],[24,0]]}]}]
      }]
    }
    Image_IconFX {
     id=#7
     possessed_class=Image
     tracks [
      {
       type=MovieSceneByteTrack
       display_name=Visibility
       sections [{type=MovieSceneByteSection, end_frame=30, start_frame=0, channels [None {type=MovieSceneScriptingByteChannel, keys [[0,3],[30,2]]}]}]
      }
      {
       type=MovieSceneFloatTrack
       display_name=Render Opacity
       sections [{type=MovieSceneFloatSection, end_frame=30, start_frame=0, channels [None {type=MovieSceneScriptingFloatChannel, keys [[0,0.25],[30,0]]}]}]
      }
     ]
    }
    Image_AlertBackground_Glow {
     id=#8
     possessed_class=Image
     tracks [
      {
       type=MovieSceneFloatTrack
       display_name=Render Opacity
       sections [{type=MovieSceneFloatSection, channels [None {type=MovieSceneScriptingFloatChannel, const [0,1]}]}]
      }
      {
       type=MovieSceneByteTrack
       display_name=Visibility
       sections [{type=MovieSceneByteSection, end_frame=30, start_frame=0, channels [None {type=MovieSceneScriptingByteChannel, keys [[0,3],[30,2]]}]}]
      }
     ]
    }
    WBP_BRGameUI_GamePhaseWidget_C_150 {
     id=#14
     possessed_class=WBP_BRGameUI_GamePhaseWidget_C
     tracks [{
       type=MovieScene2DTransformTrack
       sections [{
         type=MovieScene2DTransformSection
         channels [
          Translation.X {type=MovieSceneScriptingFloatChannel, const [0,1]}
          Translation.Y {type=MovieSceneScriptingFloatChannel, keys [[0,12],[30,0]]}
         ]
        }]
      }]
    }
    SizeBox_RemainingTime {
     id=#15
     possessed_class=SizeBox
     tracks [{
       type=MovieScene2DTransformTrack
       sections [{
         type=MovieScene2DTransformSection
         channels [
          Translation.X {type=MovieSceneScriptingFloatChannel, keys [[0,-10],[12,0]]}
          Translation.Y {type=MovieSceneScriptingFloatChannel, const [0,1]}
          Angle {type=MovieSceneScriptingFloatChannel, const [0,1]}
          Scale.X {type=MovieSceneScriptingFloatChannel, keys [[0,1.2],[12,1]]}
          Scale.Y {type=MovieSceneScriptingFloatChannel, keys [[0,1.2],[12,1]]}
          Shear.X {type=MovieSceneScriptingFloatChannel, const [0,1]}
          Shear.Y {type=MovieSceneScriptingFloatChannel, const [0,1]}
         ]
        }]
      }]
    }
    Image_IconFX_LoopingPulse {
     id=#16
     possessed_class=Image
     tracks [
      {
       type=MovieSceneFloatTrack
       display_name=Render Opacity
       sections [{type=MovieSceneFloatSection, channels [None {type=MovieSceneScriptingFloatChannel, const [0,1]}]}]
      }
      {
       type=MovieSceneByteTrack
       display_name=Visibility
       sections [{type=MovieSceneByteSection, channels [None {type=MovieSceneScriptingByteChannel, const [2,1]}]}]
      }
     ]
    }
    Image_Icon {
     id=#17
     possessed_class=Image
     tracks [
      {
       type=MovieScene2DTransformTrack
       sections [{
         type=MovieScene2DTransformSection
         channels [
          Scale.X {type=MovieSceneScriptingFloatChannel, keys [[0,1.25],[9,1]]}
          Scale.Y {type=MovieSceneScriptingFloatChannel, keys [[0,1.25],[9,1]]}
         ]
        }]
      }
      {
       type=MovieSceneWidgetMaterialTrack
       sections [{type=MovieSceneParameterSection, channels [IsTransition {type=MovieSceneScriptingFloatChannel, keys [[3,1],[12,0]]}]}]
      }
     ]
    }
   ]
   display_rate {numerator=60}
  }
 }
 AnimAlertShow_Touch {
  binding_count=11
  length_seconds=3
  object=&self:AnimAlertShow_Touch
  sequence {
   bindings [
    Text_PhaseAlert {
     id=#2
     possessed_class=TextBlock
     tracks [
      {
       type=MovieSceneByteTrack
       display_name=Visibility
       sections [{type=MovieSceneByteSection, channels [None {type=MovieSceneScriptingByteChannel, keys [[0,1],[18,4]]}]}]
      }
      {
       type=MovieSceneFloatTrack
       display_name=Render Opacity
       sections [{type=MovieSceneFloatSection, channels [None {type=MovieSceneScriptingFloatChannel, keys [[27,0],[36,1]]}]}]
      }
      {
       type=MovieScene2DTransformTrack
       sections [{
         type=MovieScene2DTransformSection
         channels [
          Translation.X {type=MovieSceneScriptingFloatChannel, keys [[21,128],[45,-10]]}
          Translation.Y {type=MovieSceneScriptingFloatChannel, default=0}
          Angle {type=MovieSceneScriptingFloatChannel, default=0}
          Scale.X {type=MovieSceneScriptingFloatChannel, default=1, keys [[18,1],[25,1.25],[71,1]]}
          Scale.Y {type=MovieSceneScriptingFloatChannel, default=1, keys [[18,1],[25,1.25],[71,1]]}
         ]
        }]
      }
     ]
    }
    SizeBox_RemainingTime {
     id=#18
     possessed_class=SizeBox
     tracks [{
       type=MovieScene2DTransformTrack
       sections [{
         type=MovieScene2DTransformSection
         channels [
          Translation.Y {type=MovieSceneScriptingFloatChannel, const [0,1]}
          Angle {type=MovieSceneScriptingFloatChannel, const [0,1]}
          Scale.X {type=MovieSceneScriptingFloatChannel, const [1,1]}
          Scale.Y {type=MovieSceneScriptingFloatChannel, const [1,1]}
          Shear.X {type=MovieSceneScriptingFloatChannel, const [0,1]}
          Shear.Y {type=MovieSceneScriptingFloatChannel, const [0,1]}
         ]
        }]
      }]
    }
    Image_Icon {
     id=#5
     possessed_class=Image
     tracks [
      {
       type=MovieScene2DTransformTrack
       sections [{
         type=MovieScene2DTransformSection
         channels [
          Scale.X {type=MovieSceneScriptingFloatChannel, keys [[0,1],[9,1],[18,1.55]]}
          Scale.Y {type=MovieSceneScriptingFloatChannel, keys [[0,1],[9,1],[18,1.55]]}
         ]
        }]
      }
      {
       type=MovieSceneWidgetMaterialTrack
       sections [{type=MovieSceneParameterSection, channels [IsTransition {type=MovieSceneScriptingFloatChannel, keys [[0,0],[6,1]]}]}]
      }
     ]
    }
    Image_MapEdgeGlow {
     id=#6
     possessed_class=Image
     tracks [
      {
       type=MovieSceneFloatTrack
       display_name=Render Opacity
       sections [{type=MovieSceneFloatSection, channels [None {type=MovieSceneScriptingFloatChannel, const [0,1]}]}]
      }
      {
       type=MovieSceneByteTrack
       display_name=Visibility
       sections [{type=MovieSceneByteSection, channels [None {type=MovieSceneScriptingByteChannel, keys [[0,1],[90,3]]}]}]
      }
      {
       type=MovieScene2DTransformTrack
       sections [{
         type=MovieScene2DTransformSection
         end_frame=136
         start_frame=90
         channels [
          Translation.X {type=MovieSceneScriptingFloatChannel, const [0,1]}
          Translation.Y {type=MovieSceneScriptingFloatChannel, const [0,1]}
          Angle {type=MovieSceneScriptingFloatChannel, const [0,1]}
          Scale.X {type=MovieSceneScriptingFloatChannel, default=1, const [-1,1]}
          Scale.Y {type=MovieSceneScriptingFloatChannel, default=0.6, const [0.6,1]}
          Shear.X {type=MovieSceneScriptingFloatChannel, const [0,1]}
          Shear.Y {type=MovieSceneScriptingFloatChannel, const [0,1]}
         ]
        }]
      }
     ]
    }
    Image_IconFX {
     id=#7
     possessed_class=Image
     tracks [
      {
       type=MovieSceneByteTrack
       display_name=Visibility
       sections [{type=MovieSceneByteSection, channels [None {type=MovieSceneScriptingByteChannel, keys [[0,1],[3,3]]}]}]
      }
      {
       type=MovieSceneFloatTrack
       display_name=Render Opacity
       sections [{type=MovieSceneFloatSection, channels [None {type=MovieSceneScriptingFloatChannel, keys [[0,0],[3,1],[90,1],[180,0.25]]}]}]
      }
      {
       type=MovieSceneWidgetMaterialTrack
       sections [{
         type=MovieSceneParameterSection
         channels [
          FirstRingAnim {type=MovieSceneScriptingFloatChannel, keys [[0,0],[120,1]]}
          SecondRingAnim {type=MovieSceneScriptingFloatChannel, keys [[18,0],[150,1]]}
          FullCircleAnim {type=MovieSceneScriptingFloatChannel, const [0,2]}
          Anim_Intensity {type=MovieSceneScriptingFloatChannel, keys [[0,0],[24,1],[60,1],[180,0]]}
         ]
        }]
      }
      {
       type=MovieScene2DTransformTrack
       sections [{
         type=MovieScene2DTransformSection
         channels [
          Scale.X {type=MovieSceneScriptingFloatChannel, keys [[0,1],[150,3]]}
          Scale.Y {type=MovieSceneScriptingFloatChannel, keys [[0,1],[150,3]]}
         ]
        }]
      }
     ]
    }
    Image_AlertBackground_Glow {
     id=#8
     possessed_class=Image
     tracks [
      {
       type=MovieSceneFloatTrack
       display_name=Render Opacity
       sections [{type=MovieSceneFloatSection, channels [None {type=MovieSceneScriptingFloatChannel, keys [[7,0],[22,1],[142,0]]}]}]
      }
      {
       type=MovieSceneByteTrack
       display_name=Visibility
       sections [{type=MovieSceneByteSection, channels [None {type=MovieSceneScriptingByteChannel, keys [[0,1],[7,3]]}]}]
      }
      {
       type=MovieScene2DTransformTrack
       sections [{
         type=MovieScene2DTransformSection
         end_frame=90
         start_frame=0
         channels [
          Translation.X {type=MovieSceneScriptingFloatChannel, const [-550,1]}
          Translation.Y {type=MovieSceneScriptingFloatChannel, const [0,1]}
          Scale.X {type=MovieSceneScriptingFloatChannel, const [-2,2]}
          Scale.Y {type=MovieSceneScriptingFloatChannel, const [-2,3]}
         ]
        }]
      }
     ]
    }
    HorizontalBox_Content {
     id=#9
     display_name=HorizontalBox_0
     possessed_class=HorizontalBox
     tracks [{
       type=MovieScene2DTransformTrack
       sections [{
         type=MovieScene2DTransformSection
         channels [
          Scale.X {type=MovieSceneScriptingFloatChannel, keys [[0,1],[24,1.1]]}
          Scale.Y {type=MovieSceneScriptingFloatChannel, keys [[0,1],[24,1.1]]}
          Shear.X {type=MovieSceneScriptingFloatChannel, default=0}
          Shear.Y {type=MovieSceneScriptingFloatChannel, default=0}
         ]
        }]
      }]
    }
    SizeBox_MapEdgeGlow_Touch {
     id=#19
     possessed_class=FortVisualAttachment
     tracks [{
       type=MovieSceneFloatTrack
       display_name=Render Opacity
       sections [{type=MovieSceneFloatSection, end_frame=180, start_frame=0, channels [None {type=MovieSceneScriptingFloatChannel, keys [[82,0],[180,1]]}]}]
      }]
    }
    Image_IconFX_LoopingPulse {
     id=#20
     possessed_class=Image
     tracks [
      {
       type=MovieSceneFloatTrack
       display_name=Render Opacity
       sections [{type=MovieSceneFloatSection, channels [None {type=MovieSceneScriptingFloatChannel, keys [[54,0],[60,1]]}]}]
      }
      {
       type=MovieSceneByteTrack
       display_name=Visibility
       sections [{type=MovieSceneByteSection, channels [None {type=MovieSceneScriptingByteChannel, const [3,1]}]}]
      }
     ]
    }
    Image_AlertBackground {
     id=#11
     possessed_class=Image
     tracks [
      {
       type=MovieSceneFloatTrack
       display_name=Render Opacity
       sections [{type=MovieSceneFloatSection, channels [None {type=MovieSceneScriptingFloatChannel, keys [[12,0],[20,1],[30,1],[116,0.5],[174,0]]}]}]
      }
      {
       type=MovieSceneByteTrack
       display_name=Visibility
       sections [{type=MovieSceneByteSection, channels [None {type=MovieSceneScriptingByteChannel, const [3,1]}]}]
      }
      {
       type=MovieScene2DTransformTrack
       sections [{
         type=MovieScene2DTransformSection
         channels [
          Translation.X {type=MovieSceneScriptingFloatChannel, const [-450,1]}
          Translation.Y {type=MovieSceneScriptingFloatChannel, const [0,1]}
          Angle {type=MovieSceneScriptingFloatChannel, const [0,1]}
          Scale.X {type=MovieSceneScriptingFloatChannel, keys [[0,0],[30,-1.4]]}
          Scale.Y {type=MovieSceneScriptingFloatChannel, keys [[0,-1.25],[15,-1.2],[90,-1]]}
          Shear.X {type=MovieSceneScriptingFloatChannel, const [0,1]}
          Shear.Y {type=MovieSceneScriptingFloatChannel, const [0,1]}
         ]
        }]
      }
      {
       type=MovieSceneWidgetMaterialTrack
       sections [{type=MovieSceneParameterSection, channels [Anim_Intensity {type=MovieSceneScriptingFloatChannel, keys [[6,0],[30,1],[174,0]]}]}]
      }
     ]
    }
   ]
   display_rate {numerator=60}
  }
 }
 AnimAlertHide_Touch {
  binding_count=10
  length_seconds=0.5
  object=&self:AnimAlertHide_Touch
  sequence {bindings [
    Text_PhaseAlert {
     id=#2
     possessed_class=TextBlock
     tracks [
      {
       type=MovieSceneByteTrack
       display_name=Visibility
       sections [{type=MovieSceneByteSection, channels [None {type=MovieSceneScriptingByteChannel, keys [[0,3],[4,1]]}]}]
      }
      {
       type=MovieSceneFloatTrack
       display_name=Render Opacity
       sections [{type=MovieSceneFloatSection, channels [None {type=MovieSceneScriptingFloatChannel, keys [[0,1],[4,0]]}]}]
      }
      {
       type=MovieScene2DTransformTrack
       sections [{
         type=MovieScene2DTransformSection
         channels [
          Translation.X {type=MovieSceneScriptingFloatChannel, default=0, keys [[0,-10],[10,128]]}
          Translation.Y {type=MovieSceneScriptingFloatChannel, default=0}
          Angle {type=MovieSceneScriptingFloatChannel, default=0}
          Scale.X {type=MovieSceneScriptingFloatChannel, keys [[0,1],[10,0.9]]}
          Scale.Y {type=MovieSceneScriptingFloatChannel, keys [[0,1],[10,0.9]]}
          Shear.X {type=MovieSceneScriptingFloatChannel, default=0}
          Shear.Y {type=MovieSceneScriptingFloatChannel, default=0}
         ]
        }]
      }
     ]
    }
    HorizontalBox_Content {
     id=#12
     display_name=HorizontalBox_0
     possessed_class=HorizontalBox
     tracks [{
       type=MovieScene2DTransformTrack
       sections [{
         type=MovieScene2DTransformSection
         channels [
          Angle {type=MovieSceneScriptingFloatChannel, default=0}
          Scale.X {type=MovieSceneScriptingFloatChannel, keys [[0,1.1],[10,1]]}
          Scale.Y {type=MovieSceneScriptingFloatChannel, keys [[0,1.1],[10,1]]}
          Shear.X {type=MovieSceneScriptingFloatChannel, default=0}
          Shear.Y {type=MovieSceneScriptingFloatChannel, default=0}
         ]
        }]
      }]
    }
    Image_IconFX_LoopingPulse {
     id=#21
     possessed_class=Image
     tracks [
      {
       type=MovieSceneFloatTrack
       display_name=Render Opacity
       sections [{type=MovieSceneFloatSection, channels [None {type=MovieSceneScriptingFloatChannel, const [0,1]}]}]
      }
      {
       type=MovieSceneByteTrack
       display_name=Visibility
       sections [{type=MovieSceneByteSection, channels [None {type=MovieSceneScriptingByteChannel, const [2,1]}]}]
      }
     ]
    }
    Image_MapEdgeGlow {
     id=#13
     possessed_class=Image
     tracks [
      {
       type=MovieSceneByteTrack
       display_name=Visibility
       sections [{type=MovieSceneByteSection, end_frame=8, start_frame=0, channels [None {type=MovieSceneScriptingByteChannel, keys [[0,4],[4,1]]}]}]
      }
      {
       type=MovieSceneFloatTrack
       display_name=Render Opacity
       sections [{type=MovieSceneFloatSection, end_frame=6, start_frame=0, channels [None {type=MovieSceneScriptingFloatChannel, const [0,2]}]}]
      }
      {
       type=MovieScene2DTransformTrack
       sections [{
         type=MovieScene2DTransformSection
         end_frame=0
         start_frame=0
         channels [
          Translation.X {type=MovieSceneScriptingFloatChannel, const [0,1]}
          Translation.Y {type=MovieSceneScriptingFloatChannel, const [0,1]}
          Angle {type=MovieSceneScriptingFloatChannel, const [0,1]}
          Scale.X {type=MovieSceneScriptingFloatChannel, const [-1,1]}
          Scale.Y {type=MovieSceneScriptingFloatChannel, const [0.6,1]}
          Shear.X {type=MovieSceneScriptingFloatChannel, const [0,1]}
          Shear.Y {type=MovieSceneScriptingFloatChannel, const [0,1]}
         ]
        }]
      }
     ]
    }
    Image_IconFX {
     id=#7
     possessed_class=Image
     tracks [
      {
       type=MovieSceneByteTrack
       display_name=Visibility
       sections [{type=MovieSceneByteSection, end_frame=10, start_frame=0, channels [None {type=MovieSceneScriptingByteChannel, keys [[0,3],[8,1]]}]}]
      }
      {
       type=MovieSceneFloatTrack
       display_name=Render Opacity
       sections [{type=MovieSceneFloatSection, end_frame=10, start_frame=0, channels [None {type=MovieSceneScriptingFloatChannel, keys [[0,0.25],[10,0]]}]}]
      }
     ]
    }
    Image_AlertBackground_Glow {
     id=#8
     possessed_class=Image
     tracks [
      {
       type=MovieSceneFloatTrack
       display_name=Render Opacity
       sections [{type=MovieSceneFloatSection, channels [None {type=MovieSceneScriptingFloatChannel, const [0,1]}]}]
      }
      {
       type=MovieSceneByteTrack
       display_name=Visibility
       sections [{type=MovieSceneByteSection, end_frame=10, start_frame=0, channels [None {type=MovieSceneScriptingByteChannel, keys [[0,3],[8,1]]}]}]
      }
     ]
    }
    WBP_BRGameUI_GamePhaseWidget_C_150 {
     id=#14
     possessed_class=WBP_BRGameUI_GamePhaseWidget_C
     tracks [{
       type=MovieScene2DTransformTrack
       sections [{
         type=MovieScene2DTransformSection
         channels [
          Translation.X {type=MovieSceneScriptingFloatChannel, const [0,1]}
          Translation.Y {type=MovieSceneScriptingFloatChannel, const [0,1]}
         ]
        }]
      }]
    }
    SizeBox_MapEdgeGlow_Touch {
     id=#22
     possessed_class=FortVisualAttachment
     tracks [{
       type=MovieSceneFloatTrack
       display_name=Render Opacity
       sections [{type=MovieSceneFloatSection, end_frame=4, start_frame=0, channels [None {type=MovieSceneScriptingFloatChannel, keys [[0,1],[4,0]]}]}]
      }]
    }
    SizeBox_RemainingTime {
     id=#23
     possessed_class=SizeBox
     tracks [{
       type=MovieScene2DTransformTrack
       sections [{
         type=MovieScene2DTransformSection
         channels [
          Translation.Y {type=MovieSceneScriptingFloatChannel, const [0,1]}
          Angle {type=MovieSceneScriptingFloatChannel, const [0,1]}
          Scale.X {type=MovieSceneScriptingFloatChannel, const [1,1]}
          Scale.Y {type=MovieSceneScriptingFloatChannel, const [1,1]}
          Shear.X {type=MovieSceneScriptingFloatChannel, const [0,1]}
          Shear.Y {type=MovieSceneScriptingFloatChannel, const [0,1]}
         ]
        }]
      }]
    }
    Image_Icon {
     id=#17
     possessed_class=Image
     tracks [
      {
       type=MovieScene2DTransformTrack
       sections [{
         type=MovieScene2DTransformSection
         channels [
          Scale.X {type=MovieSceneScriptingFloatChannel, keys [[0,1.55],[2.994,1.55],[3,1]]}
          Scale.Y {type=MovieSceneScriptingFloatChannel, keys [[0,1.55],[2.994,1.55],[3,1]]}
         ]
        }]
      }
      {
       type=MovieSceneWidgetMaterialTrack
       sections [{
         type=MovieSceneParameterSection
         end_frame=10
         start_frame=0
         channels [IsTransition {type=MovieSceneScriptingFloatChannel, keys [[1,1],[4,0]]}]
        }]
      }
     ]
    }
   ]}
 }
]
graphs [
 AdjustIconSkewingForSolos {nodes [
   AdjustIconSkewingForSolos() {is=title K2Node_FunctionEntry}
   > n1 = GetDynamicMaterial(self=Get Image_Icon() -> (Image_Icon:Image Object Reference))
   n2 = GetSquadSize!(self:Athena HUDContext Object Reference=Get AthenaHUDContext!() -> (ReturnValue:Athena HUDContext Object Reference) {is=title K2Node_GetBlueprintContextEx}) -> (ReturnValue:int=0)
   > SetScalarParameterValue(self=@n1, ParameterName=Angle, Value=SelectFloat(A=0.75, B=-0.75, bPickA="Equal (Integer)"(A=@n2, B=1) {is=title K2Node_CallFunction}))
  ]}
 EventGraph {nodes [
   n1 = HandleFadeAlert() {is=title K2Node_CustomEvent}
   > Clear and Invalidate Timer by Handle(Handle=Get AlertFadeTimerHandle() -> (AlertFadeTimerHandle:Timer Handle Structure))
   n2 = Get Animation Alert!(
    AnimGlobal:Widget Animation Object Reference=Get AnimAlertShow() -> (AnimAlertShow:Widget Animation Object Reference)
    AnimTouch:Widget Animation Object Reference=Get AnimAlertShow_Touch() -> (AnimAlertShow_Touch:Widget Animation Object Reference)
   ) -> (ReturnValue:Widget Animation Object Reference) {is=title K2Node_MacroInstance}
   > n3 = Branch(Condition=IsAnimationPlaying(InAnimation=@n2))
   > n4 = StopAnimation(InAnimation=@n2)
   Call Updated Game Phase Step!(
    execute:Exec=[@n4.then, @n3.else]
    self:WBP BRGame UI Game Phase Widget Object Reference
    NewGamePhaseStep:EAthenaGamePhaseStep Enum=Get CacheGamePhaseStep() -> (CacheGamePhaseStep:EAthenaGamePhaseStep Enum=None)
    IsGamePhaseShown:bool
   ) -> (then:Exec) {is=title K2Node_CallDelegate}
   > PlayAnimation(InAnimation=Get Animation Alert(
     AnimGlobal=Get AnimAlertHide() -> (AnimAlertHide:Widget Animation Object Reference)
     AnimTouch=Get AnimAlertHide_Touch() -> (AnimAlertHide_Touch:Widget Animation Object Reference)
    ))
   > n5 = GetGamePhaseVisualsFromStep!(execute, GamePhaseStep:EAthenaGamePhaseStep Enum=Get CacheGamePhaseStep() -> (CacheGamePhaseStep=None)) -> (
    then
    GamePhaseVisuals_MessageText_4_#24:Text
    GamePhaseVisuals_PrimaryColorName_28_#25:Name
    GamePhaseVisuals_SecondaryColorName_30_#26:Name
    GamePhaseVisuals_PrimaryColor_33_#27:Linear Color Structure
    GamePhaseVisuals_SecondaryColor_35_#28:Linear Color Structure
    GamePhaseVisuals_IconTexture_38_#29:Texture 2D Object Reference
    GamePhaseVisuals_IconColor_41_#30:Linear Color Structure
    GamePhaseVisuals_ShowIconFlashing_48_#31:bool=~
    GamePhaseVisuals_ShowTextAlert_49_#32:bool=~
   )
   > n6 = Branch(Condition=@n5.GamePhaseVisuals_ShowIconFlashing_48_#31)
   ForceStopFlashing(execute=@n6.else)
   n7 = ForceStopFlashing() {is=title K2Node_CustomEvent}
   n8 = Get bCacheFlashing() -> (bCacheFlashing:bool)
   Branch(execute=[@n7.then, @n22.else], Condition=@n8)
   > Set bCacheFlashing(bCacheFlashing:bool=false) -> (Output_Get:bool)
   n9 = Get AnimTransitionFlashing() -> (AnimTransitionFlashing:Widget Animation Object Reference)
   > PlayAnimationReverse(InAnimation=@n9)
   n10 = Event PreConstruct()
   > Branch(Condition=@n10.IsDesignTime)
   > SetGamePhaseVisualsFromStep!(
    execute
    self
    GamePhaseStep:EAthenaGamePhaseStep Enum=Get PreviewGamePhaseStep() -> (PreviewGamePhaseStep:EAthenaGamePhaseStep Enum=None)
    IsDesignTime:bool=true
   ) -> (then)
   n11 = Event OnUpdateGamePhaseIcon!() -> (OutputDelegate, then, GamePhaseStep:EAthenaGamePhaseStep Enum=None) {is=title K2Node_Event}
   > n12 = Set CacheGamePhaseStep(CacheGamePhaseStep:EAthenaGamePhaseStep Enum=@n11.GamePhaseStep) -> (Output_Get:EAthenaGamePhaseStep Enum=None)
   > n13 = Branch(Condition=OR Boolean(A="Equal (Enum)"(A:EAthenaGamePhaseStep Enum=@n12, B:EAthenaGamePhaseStep Enum=Setup), B="Equal (Enum)"(A=@n12, B=Warmup)))
   > SetVisibility(InVisibility=Hidden)
   SetVisibility(execute=@n13.else, InVisibility=HitTestInvisible)
   n14 = Get CacheGamePhaseStep() -> (CacheGamePhaseStep=None)
   > n15 = GetGamePhaseVisualsFromStep(GamePhaseStep=@n14) -> (
    GamePhaseVisuals_ShowIconFlashing_48_#31=~
    GamePhaseVisuals_ShowTextAlert_49_#32=~
   )
   > SetGamePhaseVisualsFromStep(GamePhaseStep=@n14, IsDesignTime=false)
   > n16 = Sequence()
   > n17 = Branch(Condition=@n15.GamePhaseVisuals_ShowTextAlert_49_#32)
   n18 = Get Animation Alert(AnimGlobal=Get AnimAlertHide(), AnimTouch=Get AnimAlertHide_Touch())
   > n19 = Branch(Condition=IsAnimationPlaying(InAnimation=@n18))
   > n20 = StopAnimation(InAnimation=@n18)
   PlayAnimation(execute=[@n20.then, @n19.else], InAnimation=Get Animation Alert(AnimGlobal=Get AnimAlertShow(), AnimTouch=Get AnimAlertShow_Touch()))
   > Call Updated Game Phase Step(NewGamePhaseStep=Get CacheGamePhaseStep() -> (CacheGamePhaseStep=None), IsGamePhaseShown=true)
   > n21 = Set Timer by Event(Delegate=@n1, Time=Get AlertFadeTime() -> (AlertFadeTime:double), bLooping=false)
   > Set AlertFadeTimerHandle(AlertFadeTimerHandle:Timer Handle Structure=@n21) -> (Output_Get:Timer Handle Structure)
   Branch(execute=@n17.else, Condition=Is Timer Active by Handle!(Handle:Timer Handle Structure=Get AlertFadeTimerHandle()) -> (ReturnValue:bool))
   > HandleFadeAlert() {is=title K2Node_CallFunction}
   n22 = Branch(
    execute=@n16.then_1
    Condition=OR Boolean(
     A=@n15.GamePhaseVisuals_ShowIconFlashing_48_#31
     B=@n15.GamePhaseVisuals_ShowTextAlert_49_#32
    )
   )
   > n23 = Branch(Condition=@n8)
   Set bCacheFlashing(execute=@n23.else, bCacheFlashing=true)
   > n24 = GetDynamicMaterial(self=Get Image_Icon())
   > n25 = GetDynamicMaterial(self=Get Image_IconFX() -> (Image_IconFX:Image Object Reference))
   > SetScalarParameterValue(self=[@n25, @n24], ParameterName=TimeOffset, Value=Fraction!(A:double=GetTimeSeconds!() -> (ReturnValue:double)) -> (ReturnValue:double))
   > PlayAnimationForward(InAnimation=@n9)
   Event Construct()
   > SetTouchStyle()
   > AdjustIconSkewingForSolos() {is=title K2Node_CallFunction}
  ]}
 GetAnimationAlert {nodes [
   n1 = GetOwningPlayer()
   n2 = Inputs!() -> (AnimGlobal:Widget Animation Object Reference, AnimTouch:Widget Animation Object Reference)
   Outputs!(ReturnValue:Widget Animation Object Reference=Select(
     Option 0:Widget Animation Object Reference=@n2
     Option 1:Widget Animation Object Reference=@n2.AnimTouch
     Index=IsUsingInputType!(InputType:ECommonInputType Enum=Touch, PlayerController:Player Controller Object Reference=@n1) -> (ReturnValue:bool)
    ) -> (ReturnValue:Widget Animation Object Reference))
  ]}
 Sequencer Events {nodes [SequenceEvent() {is=title K2Node_CustomEvent}]}
 SetGamePhaseVisualsFromStep {nodes [
   n1 = SetGamePhaseVisualsFromStep!() -> (then, GamePhaseStep:EAthenaGamePhaseStep Enum=None, IsDesignTime:bool=~) {is=title K2Node_FunctionEntry}
   > Branch(Condition="Not Equal (Enum)"!(A:EAthenaGamePhaseStep Enum=@n1, B:EAthenaGamePhaseStep Enum=None) -> (ReturnValue:bool))
   > n2 = GetGamePhaseVisualsFromStep(GamePhaseStep=@n1) -> (
    GamePhaseVisuals_ShowIconFlashing_48_#31=~
    GamePhaseVisuals_ShowTextAlert_49_#32=~
   )
   > n3 = GetDynamicMaterial(self=Get Image_Icon())
   > SetVectorParameterValue(self=@n3, ParameterName=BackgroundColor, Value=@n2.GamePhaseVisuals_PrimaryColor_33_#27)
   > SetTextureParameterValue(self=@n3, ParameterName=Icon, Value=@n2.GamePhaseVisuals_IconTexture_38_#29)
   > SetVectorParameterValue(self=@n3, ParameterName=IconColor, Value=@n2.GamePhaseVisuals_IconColor_41_#30)
   > n4 = Branch(Condition=@n1.IsDesignTime)
   > n5 = SetScalarParameterValue(
    self=@n3
    ParameterName=IsFlashing
    Value="To Float (Boolean)"(InBool=AND Boolean(A=@n2.GamePhaseVisuals_ShowIconFlashing_48_#31, B=@n1.IsDesignTime))
   )
   Branch(
    execute=[@n5.then, @n4.else]
    Condition=OR Boolean(A=@n2.GamePhaseVisuals_ShowTextAlert_49_#32, B=@n1.IsDesignTime)
   )
   > "SetText (Text)"(self=Get Text_PhaseAlert() -> (Text_PhaseAlert:Text Object Reference), InText=@n2)
   > GetDynamicMaterial(self=Get Image_AlertBackground_Glow() -> (Image_AlertBackground_Glow:Image Object Reference))
   > GetDynamicMaterial(self=Get Image_AlertBackground() -> (Image_AlertBackground:Image Object Reference))
   > GetDynamicMaterial(self=Get Image_MapEdgeGlow() -> (Image_MapEdgeGlow:Image Object Reference))
   > n6 = GetDynamicMaterial(self=Get Image_MapEdgeGlow_Touch() -> (Image_MapEdgeGlow_Touch:Image Object Reference))
   > SetVectorParameterValue(self=@n6, ParameterName=Color, Value=@n2.GamePhaseVisuals_PrimaryColor_33_#27)
   > n7 = GetDynamicMaterial(self=Get Image_IconFX())
   > SetVectorParameterValue(self=@n7, ParameterName=GlowColor, Value=@n2.GamePhaseVisuals_PrimaryColor_33_#27)
   > SetVectorParameterValue(self=@n7, ParameterName=BackgroundColor, Value=@n2.GamePhaseVisuals_SecondaryColor_35_#28)
  ]}
 SetTouchStyle {nodes [
   SetTouchStyle() {is=title K2Node_FunctionEntry}
   n1 = GetOwningPlayer()
   > Branch(Condition=AND Boolean(A=IsUsingCompactUI(), B=IsUsingInputType(InputType=Touch, PlayerController=@n1)))
   n2 = Get Text_PhaseAlert()
   > Set AutoWrapText!(execute, AutoWrapText:bool=true, self:Text Layout Widget Object Reference=@n2) -> (then, Output_Get:bool)
   > Set WrapTextAt!(execute, WrapTextAt:float=300, self:Text Layout Widget Object Reference=@n2) -> (then, Output_Get:float)
   > SetHorizontalAlignment(
    self:Horizontal Box Slot Object Reference=SlotAsHorizontalBoxSlot!(Widget:Widget Object Reference=Get HorizontalBox_Content() -> (HorizontalBox_Content:Horizontal Box Object Reference)) -> (ReturnValue:Horizontal Box Slot Object Reference)
    InHorizontalAlignment=HAlign_Left
   )
   n3 = Get Text_PhaseAlert()
   > SetVisibility(self=@n3, InVisibility=Collapsed)
   > Set Justification!(execute, Justification:ETextJustify Enum=Left, self:Text Layout Widget Object Reference=@n3) -> (then, Output_Get:ETextJustify Enum=Left)
   > SetPadding!(
    execute
    self:Horizontal Box Slot Object Reference=SlotAsHorizontalBoxSlot(Widget=@n3)
    InPadding_Left:float=4
    InPadding_Top:float=0
    InPadding_Right:float=0
    InPadding_Bottom:float=0
   ) -> (then)
   n4 = Get Text_PhaseAlert()
   > Set ApplyLineHeightToBottomLine!(execute, ApplyLineHeightToBottomLine:bool=false, self:Text Layout Widget Object Reference=@n4) -> (then, Output_Get:bool)
   > Set LineHeightPercentage!(execute, LineHeightPercentage:float=0.7, self:Text Layout Widget Object Reference=@n4) -> (then, Output_Get:float)
  ]}
 UpdatedGamePhaseStep {nodes [UpdatedGamePhaseStep!() -> (then, NewGamePhaseStep:EAthenaGamePhaseStep Enum=None, IsGamePhaseShown:bool=~) {is=title K2Node_FunctionEntry}]}
]
variables [
 AlertFadeTime {type=number, default=5}
 AlertFadeTimerHandle {type=TimerHandle}
 bCacheFlashing {type=boolean}
 CacheGamePhaseStep {type=EAthenaGamePhaseStep, default=None}
 PreviewGamePhaseStep {type=EAthenaGamePhaseStep, default=StormShrinking}
 UpdatedGamePhaseStep {}
]
widgets {
 info {inheritedWidgetCount=2, parentClass=/Script/FortniteUI.AthenaGamePhaseWidgetBase, rootWidgetClass=&SizeBox, widgetCount=20}
 widgets [
  SizeBox_Root {widgetClassPath=&SizeBox, widget=SizeBox_Root, properties {bOverride_HeightOverride=true, heightOverride=36}}
  Overlay_Root {widgetClassPath=&Overlay, parent=SizeBox_Root, widget=Overlay_Root, properties {pixelSnapping=Disabled}, slot {class=SizeBoxSlot}}
  SizeBox_MapEdgeGlow {
   widgetClassPath=&FortVisualAttachment
   parent=Overlay_Root
   widget=SizeBox_MapEdgeGlow
   properties {
    bOverride_HeightOverride=true
    bOverride_WidthOverride=true
    heightOverride=288
    widthOverride=96
    contentAnchor=<0.5 0.5>
    renderTransform {angle=0, scale=<1 1>, shear=<0 0>, translation=<-32 0>}
   }
   slot {class=OverlaySlot, properties {horizontalAlignment=HAlign_Right, verticalAlignment=VAlign_Center}}
  }
  Image_MapEdgeGlow {
   widgetClassPath=&Image
   bIsVariable=true
   parent=SizeBox_MapEdgeGlow
   widget=Image_MapEdgeGlow
   properties {
    visibility=Hidden
    brush {
     bIsDynamicallyLoaded=false
     drawAs=Image
     imageType=FullColor
     mirroring=NoMirror
     resourceName=None
     resourceObject=/BattleRoyaleGameUI/HUD/Minimap/Assets/M_UI_GamePhaseLabelBG
     tiling=NoTile
     imageSize=<32 32>
     margin=<ltrb 0 0 0 0>
     outlineSettings {
      bUseBrushTransparency=false
      roundingType=HalfHeightRadius
      width=0
      color {colorUseRule=UseColor_Specified, specifiedColor=<rgba 0 0 0 0>}
      cornerRadii=<0 0 0 0>
     }
     tintColor {colorUseRule=UseColor_Specified, specifiedColor=<rgba 1 1 1 1>}
     uVRegion {bIsValid=false, max=<0 0>, min=<0 0>}
    }
   }
   slot {class=SizeBoxSlot}
  }
  SizeBox_MapEdgeGlow_Touch {
   widgetClassPath=&FortVisualAttachment
   parent=Overlay_Root
   widget=SizeBox_MapEdgeGlow_Touch
   properties {
    heightOverride=160
    renderOpacity=0
    widthOverride=75
    contentAnchor=<0.5 0.5>
    renderTransform {angle=0, scale=<1 1>, shear=<0 0>, translation=<-12 0>}
   }
   slot {class=OverlaySlot, properties {horizontalAlignment=HAlign_Right, verticalAlignment=VAlign_Center}}
  }
  Image_MapEdgeGlow_Touch {
   widgetClassPath=&Image
   bIsVariable=true
   parent=SizeBox_MapEdgeGlow_Touch
   widget=Image_MapEdgeGlow_Touch
   properties {
    visibility=SelfHitTestInvisible
    brush {
     bIsDynamicallyLoaded=false
     drawAs=Image
     imageType=FullColor
     mirroring=NoMirror
     resourceName=None
     resourceObject=/FortUILibrary/UIKit/Buttons/Materials/M_UIKit_RadialGlow
     tiling=NoTile
     imageSize=<128 128>
     margin=<ltrb 0 0 0 0>
     outlineSettings {
      bUseBrushTransparency=false
      roundingType=HalfHeightRadius
      width=0
      color {colorUseRule=UseColor_Specified, specifiedColor=<rgba 0 0 0 0>}
      cornerRadii=<0 0 0 0>
     }
     tintColor {colorUseRule=UseColor_Specified, specifiedColor=<rgba 1 1 1 1>}
     uVRegion {bIsValid=false, max=<0 0>, min=<0 0>}
    }
    colorAndOpacity=<rgba 1 1 1 0.75>
   }
   slot {class=SizeBoxSlot}
  }
  Overlay_Backing {
   widgetClassPath=&Overlay
   parent=Overlay_Root
   widget=Overlay_Backing
   properties {renderTransform {angle=0, scale=<1 1>, shear=<0 0>, translation=<16 0>}}
   slot {class=OverlaySlot, properties {horizontalAlignment=HAlign_Fill, verticalAlignment=VAlign_Fill}}
  }
  Image_AlertBackground_Glow {
   widgetClassPath=&Image
   bIsVariable=true
   parent=Overlay_Backing
   widget=Image_AlertBackground_Glow
   properties {
    visibility=Hidden
    brush {
     bIsDynamicallyLoaded=false
     drawAs=Image
     imageType=FullColor
     mirroring=NoMirror
     resourceName=None
     resourceObject=&M_UI_GamePhaseLabelBG
     tiling=NoTile
     imageSize=<32 32>
     margin=<ltrb 0 0 0 0>
     outlineSettings {
      bUseBrushTransparency=false
      roundingType=HalfHeightRadius
      width=0
      color {colorUseRule=UseColor_Specified, specifiedColor=<rgba 0 0 0 0>}
      cornerRadii=<0 0 0 0>
     }
     tintColor {colorUseRule=UseColor_Specified, specifiedColor=<rgba 1 1 1 1>}
     uVRegion {bIsValid=false, max=<0 0>, min=<0 0>}
    }
    renderTransform {angle=0, scale=<0 -2>, shear=<0 0>, translation=<0 0>}
    renderTransformPivot=<1 0.5>
   }
   slot {class=OverlaySlot, properties {horizontalAlignment=HAlign_Fill, verticalAlignment=VAlign_Fill}}
  }
  Image_AlertBackground {
   widgetClassPath=&Image
   bIsVariable=true
   parent=Overlay_Backing
   widget=Image_AlertBackground
   properties {
    visibility=Hidden
    brush {
     bIsDynamicallyLoaded=false
     drawAs=Image
     imageType=FullColor
     mirroring=NoMirror
     resourceName=None
     resourceObject=/BattleRoyaleGameUI/HUD/Minimap/Assets/MI_UI_GamePhaseLabelBG
     tiling=NoTile
     imageSize=<32 32>
     margin=<ltrb 0 0 0 0>
     outlineSettings {
      bUseBrushTransparency=false
      roundingType=HalfHeightRadius
      width=0
      color {colorUseRule=UseColor_Specified, specifiedColor=<rgba 0 0 0 0>}
      cornerRadii=<0 0 0 0>
     }
     tintColor {colorUseRule=UseColor_Specified, specifiedColor=<rgba 1 1 1 1>}
     uVRegion {bIsValid=false, max=<0 0>, min=<0 0>}
    }
    renderTransform {angle=0, scale=<1 1.25>, shear=<0 0>, translation=<0 0>}
    renderTransformPivot=<1 0.5>
   }
   slot {class=OverlaySlot, properties {horizontalAlignment=HAlign_Fill, verticalAlignment=VAlign_Fill}}
  }
  HorizontalBox_Content {
   widgetClassPath=&HorizontalBox
   bIsVariable=true
   parent=Overlay_Root
   widget=HorizontalBox_Content
   properties {renderTransformPivot=<1 0.5>}
   slot {class=OverlaySlot, properties {horizontalAlignment=HAlign_Fill, verticalAlignment=VAlign_Fill}}
  }
  Text_PhaseAlert {
   widgetClassPath=&TextBlock
   bIsVariable=true
   parent=HorizontalBox_Content
   widget=Text_PhaseAlert
   properties {
    bIsVolatile=true
    justification=Center
    lineHeightPercentage=0.9
    minDesiredWidth=200
    text="Storm Shrinking!"
    textTransformPolicy=ToUpper
    visibility=Hidden
    font {
     bForceMonospaced=false
     bMaterialIsStencil=false
     fontMaterial=/BattleRoyaleGameUI/Shared/M_UI_Font_White
     fontObject=&Burbank
     letterSpacing=20
     monospacedWidth=1
     size=16.5
     skewAmount=0.2
     typefaceFontName=BurbankRegular-Black
     outlineSettings {
      bApplyOutlineToDropShadows=false
      bMiteredCorners=false
      bSeparateFillAlpha=true
      outlineMaterial=/BattleRoyaleGameUI/Shared/MI_UI_Font_SoftShadow_LargeText
      outlineSize=6
      outlineColor=<rgba 1 1 1 1>
     }
    }
    margin=<ltrb -4 -4 -8 -2>
    renderTransformPivot=<1 0.5>
    shadowOffset=<0 0>
   }
   slot {class=HorizontalBoxSlot, properties {verticalAlignment=VAlign_Center, padding=<ltrb 0 0 4 0>}}
  }
  SizeBox_RemainingTime {
   widgetClassPath=&SizeBox
   parent=HorizontalBox_Content
   widget=SizeBox_RemainingTime
   properties {bOverride_HeightOverride=true, bOverride_WidthOverride=true, heightOverride=23, widthOverride=55}
   slot {class=HorizontalBoxSlot, properties {verticalAlignment=VAlign_Center}}
  }
  Text_RemainingTime {
   widgetClassPath=&CommonTextBlock
   bInherited=true
   bIsVariable=true
   parent=SizeBox_RemainingTime
   widget=Text_RemainingTime
   properties {
    justification=Right
    lineHeightPercentage=0.9
    text=0:00
    font {
     bForceMonospaced=false
     bMaterialIsStencil=false
     fontMaterial=null
     fontObject=&BurbankSmall
     letterSpacing=0
     monospacedWidth=0.85
     size=16.5
     skewAmount=0
     typefaceFontName=Black
     outlineSettings {
      bApplyOutlineToDropShadows=false
      bMiteredCorners=false
      bSeparateFillAlpha=true
      outlineMaterial=/BattleRoyaleGameUI/Shared/MI_UI_Font_SoftShadow_SmallText
      outlineSize=3
      outlineColor=<rgba 1 1 1 1>
     }
    }
    margin=<ltrb 0 -4 0 -4>
    renderTransformPivot=<1 1>
    shadowOffset=<0 0>
   }
   slot {class=SizeBoxSlot}
  }
  Spacer_31 {widgetClassPath=&Spacer, parent=HorizontalBox_Content, widget=Spacer_31, properties {size=<4 1>}, slot {class=HorizontalBoxSlot}}
  Overlay_Icon {
   widgetClassPath=&Overlay
   parent=HorizontalBox_Content
   widget=Overlay_Icon
   slot {class=HorizontalBoxSlot, properties {horizontalAlignment=HAlign_Center, verticalAlignment=VAlign_Center}}
  }
  Image_IconFX {
   widgetClassPath=&Image
   bIsVariable=true
   parent=Overlay_Icon
   widget=Image_IconFX
   properties {
    visibility=Hidden
    brush {
     bIsDynamicallyLoaded=false
     drawAs=Image
     imageType=FullColor
     mirroring=NoMirror
     resourceName=None
     resourceObject=/BattleRoyaleGameUI/HUD/Minimap/Assets/M_UI_GamePhaseIconFX
     tiling=NoTile
     imageSize=<28 28>
     margin=<ltrb 0 0 0 0>
     outlineSettings {
      bUseBrushTransparency=false
      roundingType=HalfHeightRadius
      width=0
      color {colorUseRule=UseColor_Specified, specifiedColor=<rgba 0 0 0 0>}
      cornerRadii=<0 0 0 0>
     }
     tintColor {colorUseRule=UseColor_Specified, specifiedColor=<rgba 1 1 1 1>}
     uVRegion {bIsValid=false, max=<0 0>, min=<0 0>}
    }
    renderTransform {angle=0, scale=<12 12>, shear=<0 0>, translation=<0 0>}
   }
   slot {class=OverlaySlot, properties {horizontalAlignment=HAlign_Center, verticalAlignment=VAlign_Center}}
  }
  Image_IconFX_LoopingPulse {
   widgetClassPath=&Image
   parent=Overlay_Icon
   widget=Image_IconFX_LoopingPulse
   properties {
    visibility=Hidden
    brush {
     bIsDynamicallyLoaded=false
     drawAs=Image
     imageType=FullColor
     mirroring=NoMirror
     resourceName=None
     resourceObject=/BattleRoyaleGameUI/HUD/Minimap/Assets/MI_UI_GamePhaseIconFX_CircleOnly
     tiling=NoTile
     imageSize=<28 28>
     margin=<ltrb 0 0 0 0>
     outlineSettings {
      bUseBrushTransparency=false
      roundingType=HalfHeightRadius
      width=0
      color {colorUseRule=UseColor_Specified, specifiedColor=<rgba 0 0 0 0>}
      cornerRadii=<0 0 0 0>
     }
     tintColor {colorUseRule=UseColor_Specified, specifiedColor=<rgba 1 1 1 1>}
     uVRegion {bIsValid=false, max=<0 0>, min=<0 0>}
    }
    renderTransform {angle=0, scale=<8 8>, shear=<0 0>, translation=<0 0>}
   }
   slot {class=OverlaySlot, properties {horizontalAlignment=HAlign_Center, verticalAlignment=VAlign_Center}}
  }
  Image_Icon {
   widgetClassPath=&Image
   bIsVariable=true
   parent=Overlay_Icon
   widget=Image_Icon
   properties {
    visibility=SelfHitTestInvisible
    brush {
     bIsDynamicallyLoaded=false
     drawAs=Image
     imageType=FullColor
     mirroring=NoMirror
     resourceName=None
     resourceObject=/BattleRoyaleGameUI/HUD/Minimap/Assets/M_UI_GamePhaseIcon
     tiling=NoTile
     imageSize=<36 36>
     margin=<ltrb 0 0 0 0>
     outlineSettings {
      bUseBrushTransparency=false
      roundingType=HalfHeightRadius
      width=0
      color {colorUseRule=UseColor_Specified, specifiedColor=<rgba 0 0 0 0>}
      cornerRadii=<0 0 0 0>
     }
     tintColor {colorUseRule=UseColor_Specified, specifiedColor=<rgba 1 1 1 1>}
     uVRegion {bIsValid=false, max=<0 0>, min=<0 0>}
    }
   }
   slot {class=OverlaySlot, properties {horizontalAlignment=HAlign_Center, verticalAlignment=VAlign_Center}}
  }
  WidgetSlottingHostOverlay_RightOfGamePhase {
   widgetClassPath=/Script/WidgetSlotting.WidgetSlottingHostOverlay
   bIsVariable=true
   parent=HorizontalBox_Content
   widget=WidgetSlottingHostOverlay_RightOfGamePhase
   properties {hostTag {tagName=DynamicUI.HUD.Fort.Athena.GamePhaseStatus.StormInfoSlot}}
   slot {class=HorizontalBoxSlot}
  }
  TutorialHighlight {widgetClassPath=&TutorialHighlight}
 ]
}
