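# Minimap HUD widget for the Battle Royale game UI that shows the player's elimination count, built on the native AthenaPlayerKillsBase class.
# A CommonVisibilitySwitcher root wraps a fixed-height SizeBox holding a horizontal row: right-aligned count text, spacer and eliminations material icon.
# KillsText is an inherited variable, styled here with a soft-shadow outline material and placeholder text.
# The event graph is empty, so this Blueprint supplies only layout and style.
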WidgetBlueprint /BattleRoyaleGameUI/HUD/Minimap/WBP_BRGameUI_KillsWidget
graphs [EventGraph {links [], nodes []}]
variables []
widgets {
 info {
  inheritedWidgetCount=2
  parentClass=/Script/FortniteUI.AthenaPlayerKillsBase
  rootWidgetClass=&CommonVisibilitySwitcher
  widgetCount=7
 }
 widgets [
  CommonVisibilitySwitcher_311 {widgetClassPath=&CommonVisibilitySwitcher, widget=CommonVisibilitySwitcher_311}
  SizeBox_0 {
   widgetClassPath=&SizeBox
   parent=CommonVisibilitySwitcher_311
   widget=SizeBox_0
   properties {bOverride_HeightOverride=true, heightOverride=36}
   slot {class=CommonVisibilitySwitcherSlot}
  }
  HorizontalBox_1 {widgetClassPath=&HorizontalBox, parent=SizeBox_0, widget=HorizontalBox_1, slot {class=SizeBoxSlot}}
  KillsText {
   widgetClassPath=&TextBlock
   bInherited=true
   bIsVariable=true
   parent=HorizontalBox_1
   widget=KillsText
   properties {
    justification=Right
    text=99
    font {
     bForceMonospaced=false
     bMaterialIsStencil=false
     fontMaterial=null
     fontObject=&BurbankSmall
     letterSpacing=0
     monospacedWidth=1
     size=16.5
     skewAmount=0
     typefaceFontName=Black
     outlineSettings {
      bApplyOutlineToDropShadows=false
      bMiteredCorners=false
      bSeparateFillAlpha=true
      outlineMaterial=/BattleRoyaleGameUI/Shared/MI_UI_Font_SoftShadow_SmallText
      outlineSize=4
      outlineColor=<rgba 1 1 1 1>
     }
    }
   }
   slot {class=HorizontalBoxSlot, properties {horizontalAlignment=HAlign_Right, verticalAlignment=VAlign_Center}}
  }
  Spacer_31 {widgetClassPath=&Spacer, parent=HorizontalBox_1, widget=Spacer_31, properties {size=<2 1>}, slot {class=HorizontalBoxSlot}}
  KillsImage {
   widgetClassPath=&Image
   parent=HorizontalBox_1
   widget=KillsImage
   properties {brush {
     bIsDynamicallyLoaded=false
     drawAs=Image
     imageType=FullColor
     mirroring=NoMirror
     resourceName=None
     resourceObject=/BattleRoyaleGameUI/HUD/Minimap/Assets/MI_UI_GameStatus_Eliminations
     tiling=NoTile
     imageSize=<36 36>
     margin=<ltrb 0 0 0 0>
     outlineSettings {
      bUseBrushTransparency=false
      roundingType=HalfHeightRadius
      width=0
      color {colorUseRule=UseColor_Specified, specifiedColor=<rgba 0 0 0 0>}
      cornerRadii=<0 0 0 0>
     }
     tintColor {colorUseRule=UseColor_Specified, specifiedColor=<rgba 1 1 1 1>}
     uVRegion {bIsValid=false, max=<0 0>, min=<0 0>}
    }}
   slot {class=HorizontalBoxSlot, properties {horizontalAlignment=HAlign_Center, verticalAlignment=VAlign_Center}}
  }
  TutorialHighlight {widgetClassPath=&TutorialHighlight}
 ]
}
