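# Battle Royale minimap HUD widget, derived from AthenaPlayersLeftBase, that shows the players-left count beside a players-alive icon.
# The SetPlayersLeftText event writes the supplied text into a right-justified rich text block.
# On construct, AdjustIconSkewingForSolos reads squad size from the Athena HUD context and sets the icon dynamic material's AngleRotation parameter, using a different angle for solos than for other squad sizes.
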WidgetBlueprint /BattleRoyaleGameUI/HUD/Minimap/WBP_BRGameUI_PlayersLeftWidget
graphs [
 AdjustIconSkewingForSolos {nodes [
   AdjustIconSkewingForSolos() {is=title K2Node_FunctionEntry}
   > n1 = GetDynamicMaterial(self=Get Image_PlayerIcon() -> (Image_PlayerIcon:Image Object Reference))
   n2 = GetSquadSize!(self:Athena HUDContext Object Reference=Get AthenaHUDContext!() -> (ReturnValue:Athena HUDContext Object Reference) {is=title K2Node_GetBlueprintContextEx}) -> (ReturnValue:int=0)
   > SetScalarParameterValue(
    self=@n1
    ParameterName=AngleRotation
    Value=SelectFloat(A=0.25, B=0.75, bPickA="Equal (Integer)"(A=@n2, B=1) {is=title K2Node_CallFunction})
   )
  ]}
 EventGraph {nodes [
   n1 = Event SetPlayersLeftText!() -> (OutputDelegate, then, PlayersText:"Text (by ref)") {is=title K2Node_Event}
   > SetText(self=Get Text_PlayersLeft() -> (Text_PlayersLeft:Rich Text Block Object Reference), InText=@n1.PlayersText)
   Event Construct()
   > AdjustIconSkewingForSolos() {is=title K2Node_CallFunction}
  ]}
]
variables []
widgets {
 info {
  inheritedWidgetCount=1
  parentClass=/Script/FortniteUI.AthenaPlayersLeftBase
  rootWidgetClass=&CommonVisibilitySwitcher
  widgetCount=7
 }
 widgets [
  CommonVisibilitySwitcher_0 {widgetClassPath=&CommonVisibilitySwitcher, widget=CommonVisibilitySwitcher_0}
  SizeBox_0 {
   widgetClassPath=&SizeBox
   parent=CommonVisibilitySwitcher_0
   widget=SizeBox_0
   properties {bOverride_HeightOverride=true, heightOverride=36}
   slot {class=CommonVisibilitySwitcherSlot}
  }
  HorizontalBox_0 {widgetClassPath=&HorizontalBox, parent=SizeBox_0, widget=HorizontalBox_0, slot {class=SizeBoxSlot}}
  Text_PlayersLeft {
   widgetClassPath=/Script/UMG.RichTextBlock
   bIsVariable=true
   parent=HorizontalBox_0
   widget=Text_PlayersLeft
   properties {justification=Right, text=100, textStyleSet=/BattleRoyaleGameUI/HUD/Minimap/Assets/UIStyle_WBP_BRGameUI_PlayersLeft}
   slot {class=HorizontalBoxSlot, properties {verticalAlignment=VAlign_Center}}
  }
  Spacer_35 {widgetClassPath=&Spacer, parent=HorizontalBox_0, widget=Spacer_35, properties {size=<2 1>}, slot {class=HorizontalBoxSlot}}
  Image_PlayerIcon {
   widgetClassPath=&Image
   bIsVariable=true
   parent=HorizontalBox_0
   widget=Image_PlayerIcon
   properties {brush {
     bIsDynamicallyLoaded=false
     drawAs=Image
     imageType=FullColor
     mirroring=NoMirror
     resourceName=None
     resourceObject=/BattleRoyaleGameUI/HUD/Minimap/Assets/MI_UI_GameStatus_PlayersAlive
     tiling=NoTile
     imageSize=<36 36>
     margin=<ltrb 0 0 0 0>
     outlineSettings {
      bUseBrushTransparency=false
      roundingType=HalfHeightRadius
      width=0
      color {colorUseRule=UseColor_Specified, specifiedColor=<rgba 0 0 0 0>}
      cornerRadii=<0 0 0 0>
     }
     tintColor {colorUseRule=UseColor_Specified, specifiedColor=<rgba 1 1 1 1>}
     uVRegion {bIsValid=false, max=<0 0>, min=<0 0>}
    }}
   slot {class=HorizontalBoxSlot, properties {horizontalAlignment=HAlign_Center, verticalAlignment=VAlign_Center}}
  }
  TutorialHighlight {widgetClassPath=&TutorialHighlight}
 ]
}
