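# Battle Royale minimap HUD widget for team kills, showing the inherited KillsText count beside a team eliminations icon.
# Its parent class is AthenaPlayerTeamKillsBase.
# On Construct, it reads squad size from the Athena HUD context and collapses itself for a squad of one, otherwise setting itself HitTestInvisible.
# A leftover Delay and squad-size debug print also lead into SetVisibility, but Construct does not trigger them.
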WidgetBlueprint /BattleRoyaleGameUI/HUD/Minimap/WBP_BRGameUI_TeamKillsWidget
graphs [EventGraph {nodes [
   Delay(Duration=10)
   n1 = GetSquadSize!(self:Athena HUDContext Object Reference=Get AthenaHUDContext!() -> (ReturnValue:Athena HUDContext Object Reference) {is=title K2Node_GetBlueprintContextEx}) -> (ReturnValue:int=0)
   > n2 = PrintString!(execute, InString=Append(A="squadsize ", B="To String (Integer)"(InInt=@n1)), bPrintToScreen, bPrintToLog, TextColor, Duration, Key) -> (then)
   SetVisibility(
    execute=[@n2.then, @n3.then]
    InVisibility=Select(Option 0=HitTestInvisible, Option 1=Collapsed, Index="Equal (Integer)"(A=@n1, B=1) {is=title K2Node_CallFunction})
   )
   n3 = Event Construct()
  ]}]
variables []
widgets {
 info {inheritedWidgetCount=2, parentClass=/Script/FortniteUI.AthenaPlayerTeamKillsBase, rootWidgetClass=&SizeBox, widgetCount=6}
 widgets [
  SizeBox_0 {widgetClassPath=&SizeBox, widget=SizeBox_0, properties {bOverride_HeightOverride=true, heightOverride=36}}
  HorizontalBox_1 {widgetClassPath=&HorizontalBox, parent=SizeBox_0, widget=HorizontalBox_1, slot {class=SizeBoxSlot}}
  KillsText {
   widgetClassPath=&TextBlock
   bInherited=true
   bIsVariable=true
   parent=HorizontalBox_1
   widget=KillsText
   properties {
    justification=Right
    text=99
    font {
     bForceMonospaced=false
     bMaterialIsStencil=false
     fontMaterial=null
     fontObject=&BurbankSmall
     letterSpacing=0
     monospacedWidth=1
     size=16.5
     skewAmount=0
     typefaceFontName=Black
     outlineSettings {
      bApplyOutlineToDropShadows=false
      bMiteredCorners=false
      bSeparateFillAlpha=true
      outlineMaterial=/BattleRoyaleGameUI/Shared/MI_UI_Font_SoftShadow_SmallText
      outlineSize=4
      outlineColor=<rgba 1 1 1 1>
     }
    }
   }
   slot {class=HorizontalBoxSlot, properties {horizontalAlignment=HAlign_Right, verticalAlignment=VAlign_Center}}
  }
  Spacer_31 {widgetClassPath=&Spacer, parent=HorizontalBox_1, widget=Spacer_31, properties {size=<2 1>}, slot {class=HorizontalBoxSlot}}
  KillsImage {
   widgetClassPath=&Image
   parent=HorizontalBox_1
   widget=KillsImage
   properties {brush {
     bIsDynamicallyLoaded=false
     drawAs=Image
     imageType=FullColor
     mirroring=NoMirror
     resourceName=None
     resourceObject=/BattleRoyaleGameUI/HUD/Minimap/Assets/MI_UI_GameStatus_TeamEliminations
     tiling=NoTile
     imageSize=<36 36>
     margin=<ltrb 0 0 0 0>
     outlineSettings {
      bUseBrushTransparency=false
      roundingType=HalfHeightRadius
      width=0
      color {colorUseRule=UseColor_Specified, specifiedColor=<rgba 0 0 0 0>}
      cornerRadii=<0 0 0 0>
     }
     tintColor {colorUseRule=UseColor_Specified, specifiedColor=<rgba 1 1 1 1>}
     uVRegion {bIsValid=false, max=<0 0>, min=<0 0>}
    }}
   slot {class=HorizontalBoxSlot, properties {horizontalAlignment=HAlign_Center, verticalAlignment=VAlign_Center}}
  }
  TutorialHighlight {widgetClassPath=&TutorialHighlight}
 ]
}
