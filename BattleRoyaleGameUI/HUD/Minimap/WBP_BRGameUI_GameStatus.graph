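# Battle Royale HUD block below the minimap, stacking kills, team kills, players-left and game phase widgets.
# As a tag-based widget host, it switches ranked kills and players-left content into animated switchers, restoring defaults on removal.
# Slotted content implementing BPI_InMatchFeedbackVMProvider receives the ranked bonuses view model.
# Extension widgets join the stack with default hide tags and HUD-dependent alignment; a border spacer follows the minimap face-north setting.
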WidgetBlueprint /BattleRoyaleGameUI/HUD/Minimap/WBP_BRGameUI_GameStatus
graphs [
 AddWidgetSlotToCanvas {nodes [
   n1 = AddWidgetSlotToCanvas!() -> (then, Target:Widget Switcher Object Reference, Widget:Widget Object Reference) {is=title K2Node_FunctionEntry}
   > AddChild(self=@n1, Content=@n1.Widget)
   > SetActiveWidget(self:Widget Switcher Object Reference=@n1, Widget=@n1.Widget)
  ]}
 EventGraph {nodes [
   OnMinimapRotationModeChanged!() -> (OutputDelegate, then, bMinimapRotationModeOn:bool=~) {is=title K2Node_CustomEvent}
   n1 = Get FortHUDContext!() -> (ReturnValue:Fort HUDContext Object Reference) {is=title K2Node_GetBlueprintContextEx}
   > SetVisibility(
    self=Get Spacer_MapBorder() -> (Spacer_MapBorder:Spacer Object Reference)
    InVisibility=Select(
     Option 0=Collapsed
     Option 1=HitTestInvisible
     Index=IsPlayerAlwaysFaceNorthSettingEnabled!(self:Fort HUDContext Object Reference=@n1) -> (ReturnValue:bool)
    )
   )
   Event Construct()
   n2 = Get RankedKillsWidgetSlotTag() -> (RankedKillsWidgetSlotTag:Gameplay Tag Structure)
   > RegisterWidgetHost!(execute, WidgetHost:Widget Object Reference, WidgetHostTag:"Gameplay Tag Structure (by ref)"=@n2, IsPlayerSpecific:bool=true) -> (then)
   n3 = Get RankedPlayersLeftWidgetSlotTag() -> (RankedPlayersLeftWidgetSlotTag:Gameplay Tag Structure)
   > RegisterWidgetHost(WidgetHostTag=@n3, IsPlayerSpecific=true)
   n4 = Create Event()
   n5 = Get FortHUDContext()
   > Bind Event to On Minimap Rotation Mode Changed(self:Fort HUDContext Object Reference=@n5, Delegate=@n4) {is=title K2Node_AddDelegate}
   > OnMinimapRotationModeChanged!(execute, self, bMinimapRotationModeOn:bool=false) -> (then) {is=title K2Node_CallFunction}
   Event Destruct()
   > UnregisterWidgetHost!(execute, WidgetHost:Widget Object Reference, WidgetHostTag:"Gameplay Tag Structure (by ref)"=@n2) -> (then)
   > UnregisterWidgetHost(WidgetHostTag=@n3)
   > Unbind Event from On Minimap Rotation Mode Changed(self:Fort HUDContext Object Reference=@n5, Delegate=@n4) {is=title K2Node_RemoveDelegate}
   n6 = Event Add Content To Host Widget!() -> (
    OutputDelegate
    then
    WidgetHostTag:"Gameplay Tag Structure (by ref)"
    ContentWidget:Widget Object Reference
    Padding:"Margin Structure (by ref)"
    bFill:bool
   ) {is=title K2Node_Event}
   > Set SlottedContentWidget(SlottedContentWidget:Widget Object Reference=@n6.ContentWidget) -> (Output_Get:Widget Object Reference)
   > n7 = Branch(Condition=MatchesTag!(TagOne:Gameplay Tag Structure=@n6.WidgetHostTag, TagTwo:Gameplay Tag Structure=Get RankedKillsWidgetSlotTag(), bExactMatch:bool=false) -> (ReturnValue:bool))
   > n8 = AddWidgetSlotToCanvas!(
    execute
    self
    Target:Widget Switcher Object Reference=Get WidgetSwitcher_Kills() -> (WidgetSwitcher_Kills:Common Animated Switcher Object Reference)
    Widget:Widget Object Reference=@n6.ContentWidget
   ) -> (then) {is=title K2Node_CallFunction}
   Branch(
    execute=[@n9.then, @n8.then]
    Condition=Does Object Implement Interface!(
     TestObject:Object Reference=@n6.ContentWidget
     Interface:Interface Interface=&/BPI_InMatchFeedbackVMProvider.BPI_InMatchFeedbackVMProvider_C
    ) -> (ReturnValue:bool)
   )
   > ProvideVM!(
    execute:Exec
    self:Object Reference=@n6.ContentWidget
    VM:Athena Ranked Bonuses VM Object Reference=Get AthenaRankedBonusesVM() -> (AthenaRankedBonusesVM:Athena Ranked Bonuses VM Object Reference)
   ) -> (then:Exec) {is=title K2Node_Message}
   Branch(execute=@n7.else, Condition=MatchesTag(TagOne=@n6.WidgetHostTag, TagTwo=Get RankedPlayersLeftWidgetSlotTag(), bExactMatch=false))
   > n9 = AddWidgetSlotToCanvas(Target=Get WidgetSwitcher_PlayersLeft() -> (WidgetSwitcher_PlayersLeft:Common Animated Switcher Object Reference), Widget=@n6.ContentWidget)
   n10 = Event OnExtensionWidgetCreated!() -> (OutputDelegate, then, InUserWidget:User Widget Object Reference, InExtensionSlot:EUIExtensionSlot Enum=Invalid) {is=title K2Node_Event}
   > n11 = AddChildToVerticalBox(self=Get VBox_GameInfoBox() -> (VBox_GameInfoBox:Vertical Box Object Reference), Content=@n10.InUserWidget)
   > n12 = IsNewHUDEnabled!(execute, WorldContextObject:Object Reference=Self-Reference()) -> (then, IsNewHUD:bool)
   > SetHorizontalAlignment(
    self:Vertical Box Slot Object Reference=@n11
    InHorizontalAlignment=Select(Option 0:EHorizontalAlignment Enum=HAlign_Fill, Option 1:EHorizontalAlignment Enum=HAlign_Right, Index=@n12) -> (ReturnValue:EHorizontalAlignment Enum)
   )
   > SetVerticalAlignment(self:Vertical Box Slot Object Reference=@n11, InVerticalAlignment=VAlign_Fill)
   > n13 = Cast To FortHUDElementWidget!(execute, Object=@n10.InUserWidget) -> (then, CastFailed, AsFort HUDElement Widget:Fort HUDElement Widget Object Reference) {is=title K2Node_DynamicCast}
   > AppendHUDElementTags!(
    execute
    self:Fort HUDElement Widget Object Reference=@n13
    AdditionalTags:"Gameplay Tag Container Structure (by ref)"=Get GameInfoBoxDefaultHUDTags() -> (GameInfoBoxDefaultHUDTags:Gameplay Tag Container Structure)
   ) -> (then)
  ]}
 RemoveContentFromWidgetHost {nodes [
   n1 = Remove Content From Host Widget!() -> (
    then
    WidgetHostTag:"Gameplay Tag Structure (by ref)"
    ContentWidget:Widget Object Reference
    RemovalType:EWidgetSlottingRemovalType Enum=Unparent
   ) {is=title K2Node_FunctionEntry}
   > n2 = Branch(Condition=MatchesTag(TagOne=@n1, TagTwo=Get RankedKillsWidgetSlotTag(), bExactMatch=false))
   n3 = Get WidgetSwitcher_Kills()
   > n4 = RemoveChild(self=@n3, Content=@n1.ContentWidget)
   > n5 = Branch(Condition=@n4)
   > n6 = SetActiveWidgetIndex(self=@n3, Index=0)
   Return Node!(execute=[@n6.then, @n5.else], ReturnValue:bool=@n4)
   n7 = Branch(execute=@n2.else, Condition=MatchesTag(TagOne=@n1, TagTwo=Get RankedPlayersLeftWidgetSlotTag(), bExactMatch=false))
   n8 = Get WidgetSwitcher_PlayersLeft()
   > n9 = RemoveChild(self=@n8, Content=@n1.ContentWidget)
   > n10 = Branch(Condition=@n9)
   > n11 = SetActiveWidgetIndex(self=@n8, Index=0)
   Return Node(execute=[@n11.then, @n10.else], ReturnValue=@n9)
   Return Node(execute=@n7.else, ReturnValue=false)
  ]}
 RemoveWidgetSlotFromCanvas {nodes [
   n1 = RemoveWidgetSlotFromCanvas!() -> (then, Target:Panel Widget Object Reference, Content:Widget Object Reference) {is=title K2Node_FunctionEntry}
   > RemoveChild(self=@n1, Content=@n1.Content)
  ]}
]
variables [
 GameInfoBoxDefaultHUDTags {
  type=GameplayTagContainer
  default {gameplayTags [
    {tagName=Athena.HideForCapture}
    {tagName=HUD.Athena.Hide.WithInventory}
    {tagName=HUD.Athena.Hide.WithMap}
    {tagName=HUD.LTM.HideForTryHard}
    {tagName=HUD.Event.HideForSpecialEvent}
    {tagName=Athena.Creative.HideForLoading}
    {tagName=HUD.Hide.WhileLoading}
   ]}
 }
 RankedKillsWidgetSlotTag {type=GameplayTag, default {tagName=Hud.Athena.Habanero.RankedKills}}
 RankedPlayersLeftWidgetSlotTag {type=GameplayTag, default {tagName=HUD.Athena.Habanero.RankedPlayersLeft}}
 SlottedContentWidget {type=Widget, default=null}
]
widgets {
 info {
  inheritedWidgetCount=0
  namedSlotCount=1
  parentClass=/Script/FortniteUI.FortExtensionWidgetContainer
  rootWidgetClass=/Game/Athena/HUD/FortHUDElementWrapper.FortHUDElementWrapper_C
  widgetCount=11
 }
 widgets [
  FortHUDElementWrapper {
   widgetClassPath=/Game/Athena/HUD/FortHUDElementWrapper
   bIsVariable=true
   widget=FortHUDElementWrapper
   properties {elementTag {tagName=HUD.BelowMinimap.GameStatus}}
  }
  Overlay_1 {widgetClassPath=&Overlay, namedSlotHost=FortHUDElementWrapper, widget=Overlay_1, properties {bOverride_Cursor=true}}
  HorizontalBox_0 {
   widgetClassPath=&HorizontalBox
   parent=Overlay_1
   widget=HorizontalBox_0
   slot {class=OverlaySlot, properties {horizontalAlignment=HAlign_Right}}
  }
  VBox_GameInfoBox {
   widgetClassPath=&VerticalBox
   bIsVariable=true
   parent=HorizontalBox_0
   widget=VBox_GameInfoBox
   properties {bOverride_Cursor=true}
   slot {class=HorizontalBoxSlot}
  }
  WidgetSwitcher_Kills {
   widgetClassPath=/Script/CommonUI.CommonAnimatedSwitcher
   bIsVariable=true
   parent=VBox_GameInfoBox
   widget=WidgetSwitcher_Kills
   properties {bOverride_Cursor=true}
   slot {class=VerticalBoxSlot, properties {horizontalAlignment=HAlign_Right}}
  }
  BRGameUI_KillsWidget {
   widgetClassPath=&/WBP_BRGameUI_KillsWidget
   parent=WidgetSwitcher_Kills
   widget=BRGameUI_KillsWidget
   properties {
    bOverride_Cursor=true
    hUDElementTag {gameplayTags [
      {tagName=Athena.HideForCapture}
      {tagName=HUD.Athena.Hide.WithInventory}
      {tagName=HUD.Athena.Hide.WithMap}
      {tagName=HUD.LTM.HideForTryHard}
      {tagName=HUD.Event.HideForSpecialEvent}
      {tagName=Athena.Creative.HideForLoading}
      {tagName=HUD.Hide.WhileLoading}
      {tagName=HUD.Athena.HideForPostGame}
      {tagName=HUD.LTM.HideForCrucible}
      {tagName=Athena.LTM.Mash.Title}
      {tagName=HUD.Athena.HideforRadial}
      {tagName=HUD.Hide.CreativeJoiningLobby}
      {tagName=HUD.Athena.HideOnSpectate}
      {tagName=Athena.Creative.HUDController.HideEliminationCounter}
     ]}
   }
   slot {class=WidgetSwitcherSlot}
  }
  WBP_BRGameUI_TeamKillsWidget {
   widgetClassPath=&/WBP_BRGameUI_TeamKillsWidget
   parent=VBox_GameInfoBox
   widget=WBP_BRGameUI_TeamKillsWidget
   properties {hUDElementTag {gameplayTags [
      {tagName=Athena.HideForCapture}
      {tagName=HUD.Athena.Hide.WithInventory}
      {tagName=HUD.Athena.Hide.WithMap}
      {tagName=HUD.LTM.HideForTryHard}
      {tagName=HUD.Event.HideForSpecialEvent}
      {tagName=Athena.Creative.HideForLoading}
      {tagName=HUD.Hide.WhileLoading}
      {tagName=HUD.Athena.HideForPostGame}
      {tagName=HUD.LTM.HideForCrucible}
      {tagName=Athena.LTM.Mash.Title}
      {tagName=HUD.Athena.HideforRadial}
      {tagName=HUD.Hide.CreativeJoiningLobby}
      {tagName=HUD.Athena.HideOnSpectate}
      {tagName=Athena.Creative.HUDController.HideEliminationCounter}
     ]}}
   slot {class=VerticalBoxSlot, properties {horizontalAlignment=HAlign_Right, padding=<ltrb 0 0 0 6>}}
  }
  WidgetSwitcher_PlayersLeft {
   widgetClassPath=&CommonAnimatedSwitcher
   bIsVariable=true
   parent=VBox_GameInfoBox
   widget=WidgetSwitcher_PlayersLeft
   properties {bOverride_Cursor=true}
   slot {class=VerticalBoxSlot, properties {horizontalAlignment=HAlign_Right}}
  }
  BRGameUI_PlayersLeftWidget {
   widgetClassPath=&/WBP_BRGameUI_PlayersLeftWidget
   parent=WidgetSwitcher_PlayersLeft
   widget=BRGameUI_PlayersLeftWidget
   properties {
    bOverride_Cursor=true
    hUDElementTag {gameplayTags [
      {tagName=Athena.HideForCapture}
      {tagName=HUD.Athena.Hide.WithInventory}
      {tagName=HUD.Athena.Hide.WithMap}
      {tagName=HUD.LTM.HideForTryHard}
      {tagName=HUD.Event.HideForSpecialEvent}
      {tagName=Athena.Creative.HideForLoading}
      {tagName=HUD.Hide.WhileLoading}
      {tagName=HUD.Athena.HideForPostGame}
      {tagName=HUD.LTM.HideForCrucible}
      {tagName=Athena.LTM.Mash.Title}
      {tagName=HUD.Athena.HideforRadial}
      {tagName=HUD.Hide.CreativeJoiningLobby}
      {tagName=Athena.Creative.HUDController.HidePlayerCount}
     ]}
   }
   slot {class=WidgetSwitcherSlot}
  }
  BRGameUI_GamePhaseWidget {
   widgetClassPath=&/WBP_BRGameUI_GamePhaseWidget
   parent=VBox_GameInfoBox
   widget=BRGameUI_GamePhaseWidget
   properties {
    bOverride_Cursor=true
    elementTag {tagName=HUD.Athena.GamePhase.MinimapInfo}
    hUDElementTag {gameplayTags [
      {tagName=Athena.HideForCapture}
      {tagName=HUD.Athena.Hide.WithInventory}
      {tagName=HUD.Athena.Hide.WithMap}
      {tagName=HUD.LTM.HideForTryHard}
      {tagName=HUD.Event.HideForSpecialEvent}
      {tagName=Athena.Creative.HideForLoading}
      {tagName=HUD.Hide.WhileLoading}
      {tagName=HUD.Athena.HideForPostGame}
      {tagName=HUD.LTM.HideForCrucible}
      {tagName=Athena.LTM.Mash.Title}
      {tagName=HUD.Athena.HideforRadial}
      {tagName=Athena.Creative.HUDController.HideStormTimer}
      {tagName=HUD.Hide.CreativeJoiningLobby}
     ]}
   }
   slot {class=VerticalBoxSlot, properties {horizontalAlignment=HAlign_Right}}
  }
  Spacer_MapBorder {
   widgetClassPath=&Spacer
   bIsVariable=true
   parent=HorizontalBox_0
   widget=Spacer_MapBorder
   properties {size=<12 1>}
   slot {class=HorizontalBoxSlot}
  }
 ]
}
