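# Battle Royale minimap HUD element that shows the key bound to placing a world map marker (ping) alongside a ping icon.
# A fixed-height horizontal row pairs a KeybindWidget for PlaceMapMarkerInWorld with a material-driven icon image.
# An overlaid TutorialHighlight, scaled and offset with a square outline widget, can highlight the prompt.
# It derives from FortHUDElementWidget and has no Blueprint graph logic.
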WidgetBlueprint /BattleRoyaleGameUI/HUD/Minimap/WBP_BRGameUI_PingKey
graphs [EventGraph {links [], nodes []}]
variables []
widgets {
 info {inheritedWidgetCount=1, parentClass=&FortHUDElementWidget, rootWidgetClass=&Overlay, widgetCount=6}
 widgets [
  Overlay_1 {widgetClassPath=&Overlay, widget=Overlay_1}
  SizeBox_0 {
   widgetClassPath=&SizeBox
   parent=Overlay_1
   widget=SizeBox_0
   properties {bOverride_HeightOverride=true, heightOverride=32}
   slot {class=OverlaySlot, properties {horizontalAlignment=HAlign_Fill, verticalAlignment=VAlign_Fill}}
  }
  PingRow {widgetClassPath=&HorizontalBox, parent=SizeBox_0, widget=PingRow, slot {class=SizeBoxSlot}}
  KeybindWidget {
   widgetClassPath=/FortUILibrary/Input/KeybindWidget
   parent=PingRow
   widget=KeybindWidget
   properties {boundAction=PlaceMapMarkerInWorld, brushSize=Small}
   slot {class=HorizontalBoxSlot, properties {horizontalAlignment=HAlign_Center, verticalAlignment=VAlign_Center}}
  }
  PingIcon {
   widgetClassPath=&Image
   bIsVariable=true
   parent=PingRow
   widget=PingIcon
   properties {
    visibility=SelfHitTestInvisible
    brush {
     bIsDynamicallyLoaded=false
     drawAs=Image
     imageType=FullColor
     mirroring=NoMirror
     resourceName=None
     resourceObject=/BattleRoyaleGameUI/HUD/Minimap/Assets/MI_UI_MinimapKeyIcon_Ping
     tiling=NoTile
     imageSize=<32 32>
     margin=<ltrb 0 0 0 0>
     outlineSettings {
      bUseBrushTransparency=false
      roundingType=HalfHeightRadius
      width=0
      color {colorUseRule=UseColor_Specified, specifiedColor=<rgba 0 0 0 0>}
      cornerRadii=<0 0 0 0>
     }
     tintColor {colorUseRule=UseColor_Specified, specifiedColor=<rgba 1 1 1 1>}
     uVRegion {bIsValid=false, max=<0 0>, min=<0 0>}
    }
   }
   slot {class=HorizontalBoxSlot, properties {verticalAlignment=VAlign_Center}}
  }
  TutorialHighlight {
   widgetClassPath=&TutorialHighlight
   bIsVariable=true
   parent=Overlay_1
   widget=TutorialHighlight
   properties {
    customHighlightWidget=/Game/Athena/HUD/Tutorial/Tutorial_Outline_Square.Tutorial_Outline_Square_C
    renderTransform {angle=0, scale=<1.3 0.75>, shear=<0 0>, translation=<4 25>}
   }
   slot {class=OverlaySlot, properties {horizontalAlignment=HAlign_Fill, verticalAlignment=VAlign_Fill}}
  }
 ]
}
