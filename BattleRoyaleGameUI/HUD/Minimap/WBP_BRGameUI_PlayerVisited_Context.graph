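# Battle Royale HUD minimap widget announcing an entered or newly discovered named location or landmark, with icon and name.
# Setter functions pick the switcher's POI or landmark label, choose discovery wording and set the name.
# TriggerIntroAnim plays a fade, slide and icon-squash intro, or pauses it past its end when a debug CVar is set.
# PreConstruct adjusts alignment and attempts Korean-specific name padding for compact touch UI.
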WidgetBlueprint /BattleRoyaleGameUI/HUD/Minimap/WBP_BRGameUI_PlayerVisited_Context
animations [AnimIntro {
  binding_count=2
  length_seconds=1.5
  object=&self:AnimIntro
  sequence {bindings [
    Switcher_landmarkerPOI {
     id=#1
     possessed_class=CommonWidgetSwitcherLegacy
     tracks [
      {
       type=MovieSceneFloatTrack
       display_name=Render Opacity
       sections [{type=MovieSceneFloatSection, channels [None {type=MovieSceneScriptingFloatChannel, keys [[5,0],[25,1]]}]}]
      }
      {
       type=MovieScene2DTransformTrack
       sections [{type=MovieScene2DTransformSection, channels [Translation.X {type=MovieSceneScriptingFloatChannel, keys [[5,64],[25,0]]}]}]
      }
     ]
    }
    Image_Icon {
     id=#2
     possessed_class=Image
     tracks [
      {
       type=MovieSceneFloatTrack
       display_name=Render Opacity
       sections [{type=MovieSceneFloatSection, channels [None {type=MovieSceneScriptingFloatChannel, keys [[20,0],[21,1]]}]}]
      }
      {
       type=MovieScene2DTransformTrack
       sections [{
         type=MovieScene2DTransformSection
         channels [
          Translation.X {type=MovieSceneScriptingFloatChannel, const [0,1]}
          Translation.Y {type=MovieSceneScriptingFloatChannel, const [0,1]}
          Scale.X {type=MovieSceneScriptingFloatChannel, keys [[20,1],[22,0.75],[30,1]]}
          Scale.Y {type=MovieSceneScriptingFloatChannel, keys [[20,0.5],[22,1.5],[30,1]]}
         ]
        }]
      }
     ]
    }
   ]}
 }]
graphs [
 EventGraph {nodes [
   n1 = Event PreConstruct()
   > Parent: PreConstruct(IsDesignTime=@n1.IsDesignTime)
   > SetMobileStyle() {is=title K2Node_CallFunction}
   > UpdateMobileVisuals()
  ]}
 SetIsDiscovered {nodes [
   n1 = SetIsDiscovered!() -> (then, IsDiscovered:bool=~) {is=title K2Node_FunctionEntry}
   > "SetText (Text)"(
    self=Get Text_POIState() -> (Text_POIState:Common Text Object Reference)
    InText=Select(Option 0:Text="LOCTEXT(\"Entered\")", Option 1:Text="LOCTEXT(\"New Location!\")", Index=@n1) -> (ReturnValue:Text)
   )
   > "SetText (Text)"(
    self=Get Text_LandmarkState() -> (Text_LandmarkState:Common Text Object Reference)
    InText=Select(Option 0="LOCTEXT(\"Landmark\")", Option 1="LOCTEXT(\"New Landmark!\")", Index=@n1)
   )
  ]}
 SetIsNamedLocation {nodes [
   n1 = SetIsNamedLocation!() -> (then, IsNamedLocation:bool=~) {is=title K2Node_FunctionEntry}
   > SetActiveWidgetIndex(
    self=Get Switcher_landmarkerPOI() -> (Switcher_landmarkerPOI:Common Widget Switcher Legacy Object Reference)
    Index=Select(Option 0:int=1, Option 1:int=0, Index=@n1) -> (ReturnValue:int)
   )
  ]}
 SetLocationName {nodes [
   n1 = SetLocationName!() -> (then, LocationName:Text) {is=title K2Node_FunctionEntry}
   > "SetText (Text)"(self=Get MainLocationText() -> (MainLocationText:Common Text Object Reference), InText=@n1)
  ]}
 SetMobileStyle {nodes [
   SetMobileStyle() {is=title K2Node_FunctionEntry}
   n1 = GetOwningPlayer()
   > Branch(Condition=AND Boolean(
     A=IsUsingCompactUI()
     B=IsUsingInputType!(InputType:ECommonInputType Enum=Touch, PlayerController:Player Controller Object Reference=@n1) -> (ReturnValue:bool)
    ))
   > SetHorizontalAlignment(self:Vertical Box Slot Object Reference=SlotAsVerticalBoxSlot(Widget=Get Switcher_landmarkerPOI()), InHorizontalAlignment=HAlign_Fill)
   n2 = Get Text_LandmarkState()
   > SetHorizontalAlignment(
    self:Widget Switcher Slot Object Reference=SlotAsWidgetSwitcherSlot!(Widget:Widget Object Reference=@n2) -> (ReturnValue:Widget Switcher Slot Object Reference)
    InHorizontalAlignment=HAlign_Center
   )
   > Set Justification!(execute, Justification:ETextJustify Enum=Center, self:Text Layout Widget Object Reference=@n2) -> (then, Output_Get:ETextJustify Enum=Left)
   n3 = Get Text_POIState()
   > SetHorizontalAlignment(self=SlotAsWidgetSwitcherSlot(Widget=@n3), InHorizontalAlignment=HAlign_Center)
   > Set Justification(Justification=Center, self=@n3) -> (Output_Get=Left)
  ]}
 TriggerIntroAnim {nodes [
   TriggerIntroAnim() {is=title K2Node_FunctionEntry}
   > n1 = GetConsoleVariableBoolValue!(execute, VariableName:String=Fort.HUD.PauseAthenaVisitedAreaDisplayAnimation) -> (then, ReturnValue:bool)
   > n2 = Branch(Condition=@n1)
   n3 = Get AnimIntro() -> (AnimIntro:Widget Animation Object Reference)
   > PlayAnimation(InAnimation=@n3, StartAtTime=2)
   > PauseAnimation!(execute, self:User Widget Object Reference, InAnimation:Widget Animation Object Reference=@n3) -> (then, ReturnValue:float)
   PlayAnimation(execute=@n2.else, InAnimation=@n3)
  ]}
 UpdateMobileVisuals {nodes [
   UpdateMobileVisuals() {is=title K2Node_FunctionEntry}
   n1 = IsUsingInputType(InputType=Touch, PlayerController=GetOwningPlayer())
   > Branch(Condition=AND Boolean(A=IsUsingCompactUI(), B=@n1))
   n2 = "Equal Exactly (String)"(A=GetCurrentCulture!() -> (ReturnValue:String), B=ko)
   > SetPadding!(
    execute
    self:Vertical Box Slot Object Reference=SlotAsVerticalBoxSlot(Widget=Get MainLocationText())
    InPadding:Margin Structure=Select!(
     Option 0_Left:float=0
     Option 0_Top:float=-6
     Option 0_Right:float=0
     Option 0_Bottom:float=0
     Option 1_Left:float=0
     Option 1_Top:float=16
     Option 1_Right:float=0
     Option 1_Bottom:float=0
     Index=AND Boolean(A=@n1, B=@n2)
    ) -> (ReturnValue:Margin Structure)
   ) -> (then)
  ]}
]
variables []
widgets {
 info {inheritedWidgetCount=1, parentClass=/Script/FortniteUI.AthenaCommonPrioritizedWidget, rootWidgetClass=&VerticalBox, widgetCount=8}
 widgets [
  MainBox {widgetClassPath=&VerticalBox, widget=MainBox}
  Switcher_landmarkerPOI {
   widgetClassPath=&CommonWidgetSwitcherLegacy
   bIsVariable=true
   parent=MainBox
   widget=Switcher_landmarkerPOI
   properties {pixelSnapping=Disabled}
   slot {class=VerticalBoxSlot, properties {horizontalAlignment=HAlign_Right, verticalAlignment=VAlign_Bottom}}
  }
  Text_POIState {
   widgetClassPath=&CommonTextBlock
   bIsVariable=true
   parent=Switcher_landmarkerPOI
   widget=Text_POIState
   properties {
    bAutoCollapseWithEmptyText=true
    justification=Right
    mobileFontSizeMultiplier=1.15
    text=Entered
    textTransformPolicy=ToUpper
    colorAndOpacity {colorUseRule=UseColor_Specified, specifiedColor=<rgba 1 1 1 0.75>}
    font {
     bForceMonospaced=false
     bMaterialIsStencil=false
     fontMaterial=/BattleRoyaleGameUI/Shared/M_UI_Font_FlavorText_Master
     fontObject=&BurbankSmall
     letterSpacing=0
     monospacedWidth=1
     size=16.5
     skewAmount=0.25
     typefaceFontName=Black
     outlineSettings {
      bApplyOutlineToDropShadows=false
      bMiteredCorners=false
      bSeparateFillAlpha=true
      outlineMaterial=/BattleRoyaleGameUI/Shared/MI_UI_Font_SoftShadow_SmallText
      outlineSize=4
      outlineColor=<rgba 1 1 1 1>
     }
    }
    margin=<ltrb 0 -8 -2 -8>
    shadowOffset=<0 0>
   }
   slot {class=WidgetSwitcherSlot, properties {horizontalAlignment=HAlign_Right, padding=<ltrb 0 0 2 0>}}
  }
  Text_LandmarkState {
   widgetClassPath=&CommonTextBlock
   bIsVariable=true
   parent=Switcher_landmarkerPOI
   widget=Text_LandmarkState
   properties {
    bAutoCollapseWithEmptyText=true
    justification=Right
    mobileFontSizeMultiplier=1.15
    text=Landmark
    textTransformPolicy=ToUpper
    font {
     bForceMonospaced=false
     bMaterialIsStencil=false
     fontMaterial=&M_UI_Font_FlavorText_Master
     fontObject=&BurbankSmall
     letterSpacing=0
     monospacedWidth=1
     size=16.5
     skewAmount=0.25
     typefaceFontName=Black
     outlineSettings {
      bApplyOutlineToDropShadows=false
      bMiteredCorners=false
      bSeparateFillAlpha=true
      outlineMaterial=&MI_UI_Font_SoftShadow_SmallText
      outlineSize=4
      outlineColor=<rgba 1 1 1 1>
     }
    }
    margin=<ltrb 0 -8 -2 -8>
    shadowOffset=<0 0>
   }
   slot {class=WidgetSwitcherSlot, properties {horizontalAlignment=HAlign_Right}}
  }
  HorizontalBox_0 {
   widgetClassPath=&HorizontalBox
   parent=MainBox
   widget=HorizontalBox_0
   slot {class=VerticalBoxSlot, properties {verticalAlignment=VAlign_Bottom, size {sizeRule=Fill, value=1}}}
  }
  Image_Icon {
   widgetClassPath=&Image
   parent=HorizontalBox_0
   widget=Image_Icon
   properties {
    pixelSnapping=Disabled
    visibility=SelfHitTestInvisible
    brush {
     bIsDynamicallyLoaded=false
     drawAs=Image
     imageType=FullColor
     mirroring=NoMirror
     resourceName=None
     resourceObject=/BattleRoyaleGameUI/HUD/QuickBar/Assets/MI_UI_POI_Location_Icon
     tiling=NoTile
     imageSize=<28 28>
     margin=<ltrb 0 0 0 0>
     outlineSettings {
      bUseBrushTransparency=false
      roundingType=HalfHeightRadius
      width=0
      color {colorUseRule=UseColor_Specified, specifiedColor=<rgba 0 0 0 0>}
      cornerRadii=<0 0 0 0>
     }
     tintColor {colorUseRule=UseColor_Specified, specifiedColor=<rgba 1 1 1 1>}
     uVRegion {bIsValid=false, max=<0 0>, min=<0 0>}
    }
    renderTransformPivot=<0.5 0.95>
   }
   slot {class=HorizontalBoxSlot, properties {horizontalAlignment=HAlign_Center, verticalAlignment=VAlign_Center, padding=<ltrb 0 0 0 2>}}
  }
  MainLocationText {
   widgetClassPath=&CommonTextBlock
   bIsVariable=true
   parent=HorizontalBox_0
   widget=MainLocationText
   properties {
    bAutoCollapseWithEmptyText=true
    justification=Right
    text=Location Name
    textTransformPolicy=ToUpper
    font {
     bForceMonospaced=false
     bMaterialIsStencil=false
     fontMaterial=/BattleRoyaleGameUI/HUD/Compass/Assets/M_Font_PP_White
     fontObject=&Burbank
     letterSpacing=20
     monospacedWidth=1
     size=24
     skewAmount=0
     typefaceFontName=BurbankRegular-Black
     outlineSettings {
      bApplyOutlineToDropShadows=false
      bMiteredCorners=false
      bSeparateFillAlpha=true
      outlineMaterial=/BattleRoyaleGameUI/Shared/MI_UI_Font_SoftShadow_LargeText
      outlineSize=8
      outlineColor=<rgba 1 1 1 1>
     }
    }
    margin=<ltrb -8 -8 0 -8>
    shadowOffset=<0 0>
   }
   slot {class=HorizontalBoxSlot, properties {verticalAlignment=VAlign_Top}}
  }
  TutorialHighlight {widgetClassPath=&TutorialHighlight}
 ]
}
