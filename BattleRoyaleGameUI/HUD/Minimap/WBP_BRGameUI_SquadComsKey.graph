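# Squad comms key hint widget for the Battle Royale HUD minimap area, derived from SquadComsKey.
# A bottom-right row pairs a keybind widget, bound to PlaceMapMarkerInWorld with a forced hold display, with a squad comms icon.
# A TutorialHighlight with a custom square outline overlays the row, scaled and offset by a render transform.
# PreConstruct sets the keybind brush to the small size.
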WidgetBlueprint /BattleRoyaleGameUI/HUD/Minimap/WBP_BRGameUI_SquadComsKey
graphs [EventGraph {nodes [
   Event PreConstruct()
   > SetBrushSize!(
    execute
    self:Fort Keybind Widget Object Reference=Get KeybindWidget() -> (KeybindWidget:Keybind Widget Object Reference)
    InBrushSize:EFortBrushSize Enum=Small
   ) -> (then)
  ]}]
variables []
widgets {
 info {inheritedWidgetCount=1, parentClass=/Script/FortniteUI.SquadComsKey, rootWidgetClass=&HorizontalBox, widgetCount=6}
 widgets [
  HorizontalBox_0 {widgetClassPath=&HorizontalBox, widget=HorizontalBox_0}
  Overlay_1 {widgetClassPath=&Overlay, parent=HorizontalBox_0, widget=Overlay_1, slot {class=HorizontalBoxSlot}}
  SquadComsRow {
   widgetClassPath=&HorizontalBox
   parent=Overlay_1
   widget=SquadComsRow
   slot {class=OverlaySlot, properties {horizontalAlignment=HAlign_Right, verticalAlignment=VAlign_Bottom}}
  }
  KeybindWidget {
   widgetClassPath=/FortUILibrary/Input/KeybindWidget
   bIsVariable=true
   parent=SquadComsRow
   widget=KeybindWidget
   properties {boundAction=PlaceMapMarkerInWorld, forcedHoldKeybindStatus=ForcedHold}
   slot {class=HorizontalBoxSlot}
  }
  SquadComsIcon {
   widgetClassPath=&Image
   parent=SquadComsRow
   widget=SquadComsIcon
   properties {brush {
     bIsDynamicallyLoaded=false
     drawAs=Image
     imageType=NoImage
     mirroring=NoMirror
     resourceName=None
     resourceObject=/Game/Athena/HUD/Art/SquadCommIco
     tiling=NoTile
     imageSize=<32 32>
     margin=<ltrb 0 0 0 0>
     outlineSettings {
      bUseBrushTransparency=false
      roundingType=HalfHeightRadius
      width=0
      color {colorUseRule=UseColor_Specified, specifiedColor=<rgba 0 0 0 0>}
      cornerRadii=<0 0 0 0>
     }
     tintColor {colorUseRule=UseColor_Specified, specifiedColor=<rgba 1 1 1 1>}
     uVRegion {bIsValid=false, max=<0 0>, min=<0 0>}
    }}
   slot {class=HorizontalBoxSlot, properties {verticalAlignment=VAlign_Center}}
  }
  TutorialHighlight {
   widgetClassPath=&TutorialHighlight
   bIsVariable=true
   parent=Overlay_1
   widget=TutorialHighlight
   properties {
    customHighlightWidget=/Game/Athena/HUD/Tutorial/Tutorial_Outline_Square.Tutorial_Outline_Square_C
    renderTransform {angle=0, scale=<1.3 0.75>, shear=<0 0>, translation=<0 23>}
   }
   slot {class=OverlaySlot, properties {horizontalAlignment=HAlign_Fill, verticalAlignment=VAlign_Fill}}
  }
 ]
}
