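# Translucent UI material for the Battle Royale minimap frame: a rounded-box map backing with edge glow, outer ring and rotating cardinal-direction letters.
# The box SDF drives glow and outlines, with the ring cut away around letter shadows.
# MapRotation spins radially mirrored letter textures, and a curve shifts them along the square shape.
# Frame colors adapt to the Slate post buffer behind it, with low-end switches for some values.
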Material /BattleRoyaleGameUI/HUD/Minimap/Assets/M_UI_MinimapFrame
graph {
 properties {bAutomaticallySetUsageInEditor=false, blendMode=BLEND_Translucent, materialDomain=MD_UI, usedShadingModels=Default Lit}
 nodes [
  n1 = NamedRerouteUsage()
  n2 = ScalarParameter() {defaultValue=0.125, parameterName=GlowOpacity}
  n3 = Constant() {r=0.121}
  n4 = LinearInterpolate(
   A=Multiply(
    A=NamedRerouteUsage()
    B=MaterialFunctionCall!(Default=@n2, LowEnd=@n3) {materialFunction=/BattleRoyaleGameUI/Shared/MF_UI_LowEnd_ValueSwitcher}.Result
   )
   B=MaterialFunctionCall(Default=ScalarParameter() {defaultValue=0.125, parameterName=EdgeOutline_Opacity}, LowEnd=Constant() {r=0.121}) {materialFunction=&MF_UI_LowEnd_ValueSwitcher}.Result
   Alpha=NamedRerouteUsage()
  )
  n5 = NamedRerouteUsage()
  n6 = Saturate(None=Subtract(A=Saturate(None=@n4), B=Saturate(None=@n5)))
  n7 = LinearInterpolate(
   A=LinearInterpolate(A=@n6, B=ScalarParameter() {defaultValue=0.5, parameterName=OutlineOpacity}, Alpha=NamedRerouteUsage())
   Alpha=NamedRerouteUsage()
  )
  n8 = ScalarParameter() {defaultValue=1, parameterName=ShowCardinalDirections}
  LinearInterpolate(
   A=Multiply(A=OneMinus(None=@n1), B=LinearInterpolate(A=@n6, B=@n7, Alpha=@n8))
   B=ScalarParameter() {defaultValue=0.975, parameterName=MapOpacity}
   Alpha=NamedRerouteUsage()
  )
  n9 = MaterialFunctionCall!(BufferSource=MaterialFunctionCall!("UVs (Optional)") {materialFunction=/Engine/Functions/UserInterface/GetSlatePost0}.RGB) {materialFunction=/BattleRoyaleGameUI/Shared/MF_UI_AdaptiveColors}
  n10 = MaterialFunctionCall(Default=@n9.Dim, LowEnd=@n9.Dim) {materialFunction=&MF_UI_LowEnd_ValueSwitcher}
  n11 = NamedRerouteUsage()
  n12 = LinearInterpolate(A=@n10.Result, B=@n9.Bright, Alpha=@n11)
  n13 = Add(A=NamedRerouteUsage(), B=@n11)
  LinearInterpolate(
   A=LinearInterpolate(
    A=@n10.Result
    B=LinearInterpolate(A=@n10.Result, B=@n12, Alpha=@n13)
    Alpha=ScalarParameter() {defaultValue=1, parameterName=ShowCardinalDirections}
   )
   B=VectorParameter() {parameterName=BackgroundColor, defaultValue=<rgba 0 0.1384 0.3763 1>}.RGB
   Alpha=NamedRerouteUsage()
  )
  n14 = MaterialFunctionCall(
   Size=ScalarParameter() {defaultValue=0.8375, parameterName=MinimapScale}
   CornerRoundness=ScalarParameter() {defaultValue=0.0125, parameterName=CornerRadius}
  ) {materialFunction=&MF_UI_Box}
  NamedRerouteDeclaration(None=@n14.Fill) {name=OverlayMask, nodeColor=<rgba 0 0 0 1>}
  n15 = NamedRerouteDeclaration(None=@n14.SDF) {name=OverlayMaskSDF, nodeColor=<rgba 0 0 0 1>}
  n16 = CurveAtlasRowParameter(CurveTime=SmoothStep(Value=@n15) {constMax=0, constMin=0.25}) {
   atlas=/BattleRoyaleGameUI/Shared/UICurveAtlas_BRGameUI
   curve=/BattleRoyaleGameUI/Shared/UICurve_BRGameUI_StandardFalloffs
   parameterName=FalloffCurve
  }
  NamedRerouteDeclaration(None=@n16.A) {name=MapGlowMask, nodeColor=<rgba 0 0 0 1>}
  NamedRerouteDeclaration(None=MaterialFunctionCall(
    DistanceField=@n15
    Threshold=ScalarParameter() {defaultValue=0.0333, parameterName=EdgeOutline_Threshold}
    Smoothing=ScalarParameter() {parameterName=EdgeOutline_Smoothing}
   ) {materialFunction=&MF_UI_AntiAliasing}.Inner) {name=EdgeOutline, nodeColor=<rgba 0 0 0 1>}
  n17 = MaterialFunctionCall!(
   DistanceField=Subtract(A=NamedRerouteUsage(), B=ScalarParameter() {defaultValue=0.075, parameterName=OuterRingSpread})
   Threshold
   Thickness=ScalarParameter() {defaultValue=0.011, parameterName=OuterRingThickness}
   Alignment
   Smoothing
  ) {materialFunction=/FortUILibrary/MaterialFunctions/Shapes/Modifiers/MF_UI_ShapeOutline}
  n18 = ScalarParameter() {parameterName=MapRotation}
  n19 = Rotator(Time=Add(A=@n18) {constB=-135}) {speed=-0.0175}
  n20 = Constant() {r=1.85}
  n21 = MaterialFunctionCall(
   UVs=MaterialFunctionCall!(
    UVs=MaterialFunctionCall(UVs=@n19, Texture Scale=@n20) {materialFunction=&ScaleUVsByCenter}.UVs
    Sides=Constant() {r=4}
   ) {materialFunction=/FortUILibrary/MaterialFunctions/Basics/MF_UI_UVRadialMirror}.UVs
   Texture Scale=ScalarParameter() {defaultValue=0.07, parameterName=CardinalDirectionScale}
  ) {materialFunction=&ScaleUVsByCenter}
  n22 = Constant() {r=90}
  n23 = Modulo!(A=@n18, B=@n22)
  n24 = ConstantBiasScale(None=Frac(None=Divide(A=@n23, B=@n22))) {bias=-0.5, scale=2}
  n25 = CurveAtlasRowParameter(CurveTime=OneMinus(None=Abs(None=@n24))) {
   atlas=&UICurveAtlas_BRGameUI
   curve=/BattleRoyaleGameUI/Shared/UICurve_BRGameUI_SquareMinimap
   defaultValue=1
   parameterName=TranslationCurve
  }
  n26 = AppendVector(A=ComponentMask(None=@n21.UVs) {r=true}, B=Add(A=ComponentMask(None=@n21.UVs) {g=true}, B=LinearInterpolate(Alpha=@n25.R) {constB=2.8}))
  n27 = MaterialFunctionCall(Float4=Multiply(
    A=TextureSample(UVs=@n19) {samplerType=SAMPLERTYPE_LinearColor, texture=&/T_MinimapCardinalDirectionsMask}.RGBA
    B=TextureSample(UVs=@n26) {samplerType=SAMPLERTYPE_LinearColor, texture=&/T_MinimapCardinalDirections}.RGBA
   )) {materialFunction=&BreakOutFloat4Components}
  n28 = Max(A=Max(A=@n27.R, B=@n27.R), B=@n27.B)
  n29 = MaterialFunctionCall!(DistanceField=Add(A=Max(A=@n28, B=@n27.A)) {constB=0.2}, Threshold=Constant() {r=0.6}, Smoothing) {materialFunction=/BattleRoyaleGameUI/Shared/MF_UI_DFToMaskAndShadow}
  n30 = Saturate(None=Multiply(A=@n29.Shadow) {constB=10})
  NamedRerouteDeclaration(None=Saturate(None=Subtract(A=@n17.Outline, B=@n30))) {name=OutlineMask, nodeColor=<rgba 0 0 0 1>}
  NamedRerouteDeclaration(None=@n29.Mask) {name=CardinalLettersMask, nodeColor=<rgba 0 0 0 1>}
 ]
}
